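; Nitro Drome center jumbotron actor that shows the reward countdown and event reaction screens.
; On clients, global reward-status events bind it once to the playspace event router and control a one-second timer while the reward is active.
; The timer packs minute and second digits into screen A's custom primitive data.
; Event tags look up a texture and color index for screen B, passing game time as custom primitive data.
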(event Custom|NitroDromeJumboTronScreenTest
  ;[0 0.8622 1] Test Screens
  (bind _keys (Utilities|Map|Keys (Variables|Default|GetTagTextureMap)))
  (bind _output_get (Variables|Default|SetPayload%20Event%20Tag (Utilities|Array|RandomArrayItem _keys)))
  (CallFunction|SetMaterialDataandActivateScreen _output_get))

(event Custom|SetMaterialDataAndActivateScreen
  (params (Event%20Tag :type Gameplay%20Tag%20Structure))
  (bind _tagtexturemap (Variables|Default|GetTagTextureMap))
  (bind (_value _returnvalue) (Utilities|Map|Find _tagtexturemap Event%20Tag))
  (bind _tagcolormap (Variables|Default|GetTagColorMap))
  (bind (_value_2 _returnvalue_2) (Utilities|Map|Find _tagcolormap Event%20Tag))
  (if _returnvalue
    (Rendering|Material|SetTextureParameterValue (Class|BTechArtGlobalLibrary|GetDynamicMaterialInstance (Variables|Default|GetSM_NitroDrome_Screen_B) 1) "ReactionTexture" _value)
    (Rendering|Material|SetCustomPrimitiveDataFloat (Variables|Default|GetSM_NitroDrome_Screen_B) 1 (Utilities|Time|GetGameTimeinSeconds))
    (if _returnvalue_2
      (Rendering|Material|SetCustomPrimitiveDataFloat (Variables|Default|GetSM_NitroDrome_Screen_B) 2 (Math|Conversions|ToFloat(Integer) _value_2)))))

(event Custom|TryBindToNitroDromeEvents
  (bind _nitrodromeeventrouter (CallFunction|TryGetNitroDromeRouter))
  (Utilities|IsValid _nitrodromeeventrouter
    (:Is%20Valid
      (Utilities|FlowControl|DoOnce false
        (:Completed
          (bind (_asynctaskproxy _payload) (GameplayEventRouter|StartListeningToEvent _nitrodromeeventrouter "/Script/NitroDromeRuntime.OnNitroDromeEventTriggered")
            (:then
              (Fortnite|FortShippingLog self "Bound to NitroDromeEventTriggered Successfully (This means we found an event router inside the Nitro Drome Playspace)"))
            (:OnEventReceived
              (bind _payload_eventtag (Utilities|Struct|BreakOnNitroDromeEventTriggered _payload))
              (bind _output_get (Variables|Default|SetPayload%20Event%20Tag _payload_eventtag))
              (Networking|CanExecuteCosmeticEvents
                (:True
                  (CallFunction|SetMaterialDataandActivateScreen _output_get))))))))))

(event EventBeginPlay
  ;[1 1 1] Using the global state change event as reference for when the playspace is relevant to the jumbotron actor
  (Networking|SwitchHasAuthority
    (:Remote
      (bind (_asynctaskproxy _payload) (GameplayEventRouter|StartListeningToStatefulEvent (GameplayEventRouter|GetGlobalEventRouter) "/Script/NitroDromeRuntime.OnDromeRewardStatusChanged")
        (:OnEventReceived
          _)
        (:OnSavedEventStateExists
          _))
      (CallFunction|TryBindtoNitroDromeEvents) ; Always try to bind to events, even if we enter\nrelevancy when the Reward has spawned so we can see\nreactions to the Nitro Queen. We have a do once inside here\nso this will only be called once.
      (switch Utilities|FlowControl|Switch|SwitchonENitroDromeRewardStatus (Utilities|Struct|BreakOnDromeRewardStatusChanged _payload)
        (:Inactive
          _)
        (:Active
          (CallFunction|StartQueryingTimeRemaining (.RewardManager (Utilities|Struct|BreakOnDromeRewardStatusChanged _payload))))
        (:Paused
          _)
        (:Finished
          _))
      (CallFunction|StopQueryingTimeRemaining))))

(event Collision|EventActorBeginOverlap
  (params (OtherActor :type Actor%20Object%20Reference)))

(event EventTick
  (params (DeltaSeconds :type Float%20%28single-precision%29)))

(fn GetDigitAtIndexReverse
  (params (Number :type Integer) (DigitIndex :type Integer))
  (return (Math|Integer|%25%28Integer%29 (Math|Float|Truncate (/ Number (Math|Float|SelectFloat (Math|Float|Power 1e+01 (Math|Conversions|ToFloat(Integer) DigitIndex)) 1.0 (> (Math|Float|Power 1e+01 (Math|Conversions|ToFloat(Integer) DigitIndex)) 0)))) 10)))

(fn StartQueryingTimeRemaining
  (params (RewardManager :type Fort%20Nitro%20Drome%20Reward%20Manager%20Object%20Reference))
  (Variables|Default|SetRewardManager RewardManager)
  (bind _returnvalue (Utilities|Time|SetTimerbyEvent (EventDispatchers|CreateEvent :!event UpdateTimerCosmetics) 1.0 true))
  (Variables|Default|SetUpdateTimerHandle _returnvalue)
  (CallFunction|UpdateTimerCosmetics))

(fn StopQueryingTimeRemaining
  (Utilities|Time|ClearandInvalidateTimerbyHandle (Variables|Default|GetUpdateTimerHandle))
  (Variables|Default|SetRewardManager 0))

(fn TryGetNitroDromeRouter
  (bind _nitrodromeeventrouter (Variables|Default|GetNitroDromeEventRouter))
  (Utilities|IsValid _nitrodromeeventrouter
    (:Is%20Valid
      (return _nitrodromeeventrouter))
    (:Is%20Not%20Valid
      (bind _returnvalue (GameplayEventRouter|GetEventRouter "/NitroDrome/Gameplay/Actors/BP_NitroDromePlayspace.BP_NitroDromePlayspace_C" self))
      (bind _output_get (Variables|Default|SetNitroDromeEventRouter _returnvalue))
      (return _output_get))))

(fn UpdateTimerCosmetics
  (bind _rewardmanager (Variables|Default|GetRewardManager))
  (Utilities|IsValid _rewardmanager
    (:Is%20Valid
      (Rendering|Material|SetCustomPrimitiveDataVector4 (Variables|Default|GetSM_NitroDrome_Screen_A) 0 (Math|Vector4|MakeVector4 (Math|Conversions|ToFloat(Integer) (CallFunction|GetDigitatIndexReverse (Math|Timespan|GetMinutes (Math|Timespan|FromSeconds (Utilities|Time|GetTimerRemainingTimebyHandle (Class|FortNitroDromeRewardManager|GetRewardTimerHandle _rewardmanager)))) 1)) (Math|Conversions|ToFloat(Integer) (CallFunction|GetDigitatIndexReverse (Math|Timespan|GetMinutes (Math|Timespan|FromSeconds (Utilities|Time|GetTimerRemainingTimebyHandle (Class|FortNitroDromeRewardManager|GetRewardTimerHandle _rewardmanager)))))) (Math|Conversions|ToFloat(Integer) (CallFunction|GetDigitatIndexReverse (Math|Timespan|GetSeconds (Math|Timespan|FromSeconds (Utilities|Time|GetTimerRemainingTimebyHandle (Class|FortNitroDromeRewardManager|GetRewardTimerHandle _rewardmanager)))) 1)) (Math|Conversions|ToFloat(Integer) (CallFunction|GetDigitatIndexReverse (Math|Timespan|GetSeconds (Math|Timespan|FromSeconds (Utilities|Time|GetTimerRemainingTimebyHandle (Class|FortNitroDromeRewardManager|GetRewardTimerHandle _rewardmanager)))))))))))

(fn ConstructionScript
  (|CallParentFunction))
