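# Material for a motorized spotlight strip mesh, flagged for Niagara mesh particles among other mesh types, using Default Lit shading.
# Material attributes take diffuse and normal texture parameters for base color and normal.
# Emissive multiplies particle color by the red channel of a mask texture, so the particle drives the light color.
# The Color, HueShift and Brightness chain and the collection-parameter color lookup are left unconnected to the output.
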Material /NitroDrome/Assets/VFX/Materials/M_SpotStrip_x4_PG_FX
graph {
 properties {
  bUseMaterialAttributes=true
  bUsedWithMeshDeformer=true
  bUsedWithNanite=true
  bUsedWithNiagaraMeshParticles=true
  bUsedWithSkeletalMesh=true
  bUsedWithStaticLighting=true
  numCustomizedUVs=2
  pixelDepthOffsetMode=PDOM_Legacy
  usedShadingModels=Default Lit
 }
 nodes [
  ComponentMask(None=CollectionParameter() {collection=/Game/Athena/Environments/Festivus/MotorizedLights/Lighting_MPC, parameterName=SpotStrip_01_Color}) {b=true, g=true, r=true}
  n1 = VectorParameter() {parameterName=Color, defaultValue=<rgba 1 0 0 0>}
  Multiply(
   A=MaterialFunctionCall(Hue Shift Percentage=ScalarParameter() {parameterName=HueShift}, Texture=Multiply(A=@n1.RGB) {constB=10}) {materialFunction=&HueShift}.Result
   B=Multiply(A=ScalarParameter() {defaultValue=1, parameterName=Brightness}) {constB=150}
  ) {constB=150}
  n2 = TextureSampleParameter2D() {parameterName=Normal, samplerType=SAMPLERTYPE_Normal, texture=/Game/Athena/Environments/Festivus/MotorizedLights/T_MotorizedSpotLight_01_N}
  SetMaterialAttributes!(
   MaterialAttributes
   Base Color=TextureSampleParameter2D() {parameterName=Diffuse, texture=/Game/Athena/Environments/Festivus/MotorizedLights/T_MotorizedSpotLight_01_D}.RGB
   Emissive Color=Multiply(A=ParticleColor().RGB, B=TextureSample() {texture=/Game/Athena/Environments/Festivus/MotorizedLights/T_MotorizedSpotLight_01_M}.R)
   Normal=@n2.RGB
  ) {attributeSetTypes [#1, #2, #3]}
 ]
}
