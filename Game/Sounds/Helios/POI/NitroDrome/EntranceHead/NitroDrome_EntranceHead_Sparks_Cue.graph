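# Sound cue for the spark burst of the NitroDrome POI entrance head, an ambient sound with its own attenuation asset.
# A mixer plays three layers at once, each a random wave pick through a pitch modulator: sparks with varied volume, sparks pitched down, and a roar.
# Both spark layers draw on the same eight waves, so the lowered copy thickens the burst.
# Only the roar layer, with four variants, gets a short random delay.
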SoundCue /Game/Sounds/Helios/POI/NitroDrome/EntranceHead/NitroDrome_EntranceHead_Sparks_Cue
root=Mixer_0
nodes [
 {id=Delay_0, children [Modulator_1], properties {delayMax=0.15}}
 {id=Delay_1, children [Modulator_0]}
 {id=Delay_2, children [Modulator_2]}
 {id=Mixer_0, children [Delay_1,Delay_2,Delay_0], properties {inputVolume [1,1,1]}}
 {id=Modulator_0, children [Random_0], properties {pitchMax=1.2, pitchMin=0.9, volumeMax=1, volumeMin=0.7}}
 {id=Modulator_1, children [Random_1], properties {pitchMax=1.2, pitchMin=0.9, volumeMax=1, volumeMin=1}}
 {id=Modulator_2, children [Random_2], properties {pitchMax=0.85, pitchMin=0.7, volumeMax=1, volumeMin=1}}
 {
  id=Random_0
  children [WavePlayer_0,WavePlayer_1,WavePlayer_2,WavePlayer_3,WavePlayer_4,WavePlayer_5,WavePlayer_6,WavePlayer_7]
  properties {bShouldExcludeFromBranchCulling=true, weights [1,1,1,1,1,1,1,1]}
 }
 {
  id=Random_1
  children [WavePlayer_8, WavePlayer_9, WavePlayer_10, WavePlayer_11]
  properties {bShouldExcludeFromBranchCulling=true, weights [1,1,1,1]}
 }
 {
  id=Random_2
  children [WavePlayer_12, WavePlayer_13, WavePlayer_14, WavePlayer_15, WavePlayer_16, WavePlayer_17, WavePlayer_18, WavePlayer_19]
  properties {bShouldExcludeFromBranchCulling=true, weights [1,1,1,1,1,1,1,1]}
 }
 {id=WavePlayer_0, properties {soundWaveAssetPtr=&/NitroDrome_EntranceHead_Sparks_01}}
 {id=WavePlayer_1, properties {soundWaveAssetPtr=&/NitroDrome_EntranceHead_Sparks_02}}
 {id=WavePlayer_10, properties {soundWaveAssetPtr=&/NitroDrome_EntranceHead_Roar_03}}
 {id=WavePlayer_11, properties {soundWaveAssetPtr=&/NitroDrome_EntranceHead_Roar_04}}
 {id=WavePlayer_12, properties {soundWaveAssetPtr=&NitroDrome_EntranceHead_Sparks_01}}
 {id=WavePlayer_13, properties {soundWaveAssetPtr=&NitroDrome_EntranceHead_Sparks_02}}
 {id=WavePlayer_14, properties {soundWaveAssetPtr=&/NitroDrome_EntranceHead_Sparks_03}}
 {id=WavePlayer_15, properties {soundWaveAssetPtr=&/NitroDrome_EntranceHead_Sparks_04}}
 {id=WavePlayer_16, properties {soundWaveAssetPtr=&/NitroDrome_EntranceHead_Sparks_05}}
 {id=WavePlayer_17, properties {soundWaveAssetPtr=&/NitroDrome_EntranceHead_Sparks_06}}
 {id=WavePlayer_18, properties {soundWaveAssetPtr=&/NitroDrome_EntranceHead_Sparks_07}}
 {id=WavePlayer_19, properties {soundWaveAssetPtr=&/NitroDrome_EntranceHead_Sparks_08}}
 {id=WavePlayer_2, properties {soundWaveAssetPtr=&NitroDrome_EntranceHead_Sparks_03}}
 {id=WavePlayer_3, properties {soundWaveAssetPtr=&NitroDrome_EntranceHead_Sparks_04}}
 {id=WavePlayer_4, properties {soundWaveAssetPtr=&NitroDrome_EntranceHead_Sparks_05}}
 {id=WavePlayer_5, properties {soundWaveAssetPtr=&NitroDrome_EntranceHead_Sparks_06}}
 {id=WavePlayer_6, properties {soundWaveAssetPtr=&NitroDrome_EntranceHead_Sparks_07}}
 {id=WavePlayer_7, properties {soundWaveAssetPtr=&NitroDrome_EntranceHead_Sparks_08}}
 {id=WavePlayer_8, properties {soundWaveAssetPtr=&/NitroDrome_EntranceHead_Roar_01}}
 {id=WavePlayer_9, properties {soundWaveAssetPtr=&/NitroDrome_EntranceHead_Roar_02}}
]
properties {
 attenuationSettings=&/NitroDrome_EntranceHead_Att
 maxDistance=12000
 soundClassObject=&Fort_Ambient_Athena
 volumeMultiplier=0.952
}
