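# Niagara reaction effect for NitroDrome that emits short-lived red-orange particles in response to the emote value in the NitroDrome parameter collection.
# GPU mesh and sprite emitters scale spawn rates by that value, floored at a system minimum.
# Spawn-group masks place particles on a flattened torus and rotated sphere arcs swept by loop age.
# Camera-distance remaps adjust mesh scale and material parameters; a multicolor VehicleEntered emitter is disabled.
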NiagaraSystem /NitroDrome/Assets/VFX/NS_NitroDrome_Reaction_Emote
dependencies {
 usedDataInterfaces [&NiagaraDataInterfaceColorCurve, &NiagaraDataInterfaceCurve, &NiagaraDataInterfaceArrayColor]
 usedDynamicInputs [
  &MultiplyQuaternion
  &MakeQuatFromAxisAngle
  &Lerp_Float
 ]
 usedModules [
  &SystemState
  &self:SystemSpawnScript.SystemScriptSource.SystemScriptGraph.NiagaraNodeAssignment_0.SetVariables_#1
 ]
}
emitters {
 FN_Sprite {
  simTarget=GPUComputeSim
  emitterUpdateScript [
   &EmitterState()
   &SpawnRate(
    SpawnRate=&Multiply_Float(A=100, B=&ClampFloat(Float=$NPC.NPC_NitroDrome.NitroDrome_Emote, Min=$System.MinSpawnRate))
    Use Spawn Probability=false
   )
   &SpawnRate(
    SpawnGroup=1
    SpawnRate=&Multiply_Float(A=15, B=&ClampFloat(Float=$NPC.NPC_NitroDrome.NitroDrome_Emote, Min=$System.MinSpawnRate))
    Use Spawn Probability=false
   ) {module=SpawnRate001}
   &SpawnRate(
    Spawn Probability=2
    SpawnGroup=2
    SpawnRate=&Multiply_Float(A=15, B=&ClampFloat(Float=$NPC.NPC_NitroDrome.NitroDrome_Emote, Min=$System.MinSpawnRate))
    Use Spawn Probability=false
   ) {enabled=false, module=SpawnRate002}
   &SpawnRate(
    Spawn Probability=2
    SpawnGroup=3
    SpawnRate=&Multiply_Float(A=15, B=&ClampFloat(Float=$NPC.NPC_NitroDrome.NitroDrome_Emote, Min=$System.MinSpawnRate))
    Use Spawn Probability=false
   ) {module=SpawnRate003}
   &SpawnRate(
    Spawn Probability=2
    SpawnGroup=4
    SpawnRate=&Multiply_Float(A=15, B=&ClampFloat(Float=$NPC.NPC_NitroDrome.NitroDrome_Emote, Min=$System.MinSpawnRate))
    Use Spawn Probability=false
   ) {module=SpawnRate004}
   &SpawnRate(
    Spawn Probability=2
    SpawnGroup=5
    SpawnRate=&Multiply_Float(A=15, B=&ClampFloat(Float=$NPC.NPC_NitroDrome.NitroDrome_Emote, Min=$System.MinSpawnRate))
    Use Spawn Probability=false
   ) {enabled=false, module=SpawnRate005}
   &SpawnRate(
    Spawn Probability=2
    SpawnGroup=6
    SpawnRate=&Multiply_Float(A=15, B=&ClampFloat(Float=$NPC.NPC_NitroDrome.NitroDrome_Emote, Min=$System.MinSpawnRate))
    Use Spawn Probability=false
   ) {module=SpawnRate006}
  ]
  particleSpawnScript [
   &InitializeParticle(
    Color=&ColorFromCurve(
     ColorCurve=DataInterface() {propertyValues {
       AlphaCurve {keys [<at 0 1 RCIM_Cubic>,<at 0.9977 1 RCIM_Cubic>]}
       BlueCurve {keys [<at 0.2936 0>,<at 0.608 0>]}
       GreenCurve {keys [<at 0.2936 0.1016>,<at 0.608 0.2592>]}
       RedCurve {keys [<at 0.2936 1>,<at 0.608 1>]}
      }}
     CurveIndex=&RandomRangeFloat(Minimum=0, Maximum=1)
    )
    Lifetime Max=1
    Lifetime Mode=::NewEnumerator1
    Mesh Scale Mode=::NewEnumerator1
    Mesh Uniform Scale Max=250
    Mesh Uniform Scale Min=50
    Sprite Size Mode=::NewEnumerator1
    Uniform Sprite Size Max=300
    Uniform Sprite Size Min=250
    UsePositionOffset=false
    Write Material Random=true
   ) {module=InitializeParticle001}
   &AddVelocityFromPoint(
    Module.Use Falloff Distance=false
    Origin Offset=&UniformRangedVector(Minimum=<0 0 -100>, Maximum=<0 0 -100>)
    Origin Offset Coordinate Space=::Simulation
    Velocity Strength=&UniformRangedFloat(Minimum=10, Maximum=15)
   ) {enabled=false}
   &ShapeLocation(
    Handle Radius=0
    Large Radius=1000
    MaskSpawnGroup=true
    Non Uniform Scale=<2.5 1.5 0>
    Shape Primitive=::NewEnumerator3
    Torus Distribution Mode:ENiagaraTorusDistributionMode=::NewEnumerator1
    UseOwnerScale=true
   ) {module=ShapeLocation001}
   &ShapeLocation(
    MaskSpawnGroup=true
    Non Uniform Scale=<1.3 1 0.5>
    Offset=<0 -500 -2400>
    Offset Mode=::NewEnumerator0
    Rotation Angle=0.325
    Rotation Axis=<0 0 1>
    Rotation Coordinate Space=::Simulation
    Rotation Mode=::NewEnumerator1
    Spawn Group Mask=1
    Sphere Distribution=::NewEnumerator1
    Sphere Radius=5000
    U Position=&RandomRangeFloat(Minimum=0, Maximum=0)
    UseOwnerScale=true
    V Position=&FloatFromCurve(FloatCurve=DataInterface() {propertyValues {Curve {keys [<at 0 0.8>,<at 1 0.99>]}}}, CurveIndex=$System.NormalizedLoopAge)
   ) {module=ShapeLocation002}
   &ShapeLocation(
    MaskSpawnGroup=true
    Non Uniform Scale=<1.6 1 0.5>
    Offset=<0 250 -2400>
    Offset Mode=::NewEnumerator0
    Rotation Angle=0.507
    Rotation Axis=<0 0 1>
    Rotation Coordinate Space=::Simulation
    Rotation Mode=::NewEnumerator1
    Spawn Group Mask=2
    Sphere Distribution=::NewEnumerator1
    Sphere Radius=5000
    U Position=&RandomRangeFloat(Minimum=0, Maximum=0)
    UseOwnerScale=true
    V Position=&FloatFromCurve(FloatCurve=DataInterface() {propertyValues {Curve {keys [<at 0 0.7>,<at 1 0.99>]}}}, CurveIndex=$System.NormalizedLoopAge)
   ) {module=ShapeLocation003}
   &ShapeLocation(
    MaskSpawnGroup=true
    Non Uniform Scale=<1.2 1 0.5>
    Offset=<0 250 -2400>
    Offset Mode=::NewEnumerator0
    Rotation Angle=0.67
    Rotation Axis=<0 0 1>
    Rotation Coordinate Space=::Simulation
    Rotation Mode=::NewEnumerator1
    Spawn Group Mask=3
    Sphere Distribution=::NewEnumerator1
    Sphere Radius=5000
    U Position=&RandomRangeFloat(Minimum=0, Maximum=0)
    UseOwnerScale=true
    V Position=&FloatFromCurve(FloatCurve=DataInterface() {propertyValues {Curve {keys [<at 0 0.7>,<at 1 0.99>]}}}, CurveIndex=$System.NormalizedLoopAge)
   ) {module=ShapeLocation004}
   &ShapeLocation(
    MaskSpawnGroup=true
    Non Uniform Scale=<1.2 1 0.5>
    Offset=<0 250 -2400>
    Offset Mode=::NewEnumerator0
    Rotation Angle=0.82
    Rotation Axis=<0 0 1>
    Rotation Coordinate Space=::Simulation
    Rotation Mode=::NewEnumerator1
    Spawn Group Mask=4
    Sphere Distribution=::NewEnumerator1
    Sphere Radius=5000
    U Position=&RandomRangeFloat(Minimum=0, Maximum=0)
    UseOwnerScale=true
    V Position=&FloatFromCurve(FloatCurve=DataInterface() {propertyValues {Curve {keys [<at 0 0.5>,<at 1 0.99>]}}}, CurveIndex=$System.NormalizedLoopAge)
   ) {module=ShapeLocation005}
   &ShapeLocation(
    MaskSpawnGroup=true
    Non Uniform Scale=<1.6 1 0.5>
    Offset=<0 250 -2500>
    Offset Mode=::NewEnumerator0
    Rotation Angle=0.995
    Rotation Axis=<0 0 1>
    Rotation Coordinate Space=::Simulation
    Rotation Mode=::NewEnumerator1
    Spawn Group Mask=5
    Sphere Distribution=::NewEnumerator1
    Sphere Radius=5000
    U Position=&RandomRangeFloat(Minimum=0, Maximum=0)
    UseOwnerScale=true
    V Position=&FloatFromCurve(FloatCurve=DataInterface() {propertyValues {Curve {keys [<at 0 0.5>,<at 1 0.99>]}}}, CurveIndex=$System.NormalizedLoopAge)
   ) {module=ShapeLocation006}
   &ShapeLocation(
    MaskSpawnGroup=true
    Non Uniform Scale=<1.3 1 0.5>
    Offset=<0 -500 -2500>
    Offset Mode=::NewEnumerator0
    Rotation Angle=0.18
    Rotation Axis=<0 0 1>
    Rotation Coordinate Space=::Simulation
    Rotation Mode=::NewEnumerator1
    Spawn Group Mask=6
    Sphere Distribution=::NewEnumerator1
    Sphere Radius=5000
    U Position=&RandomRangeFloat(Minimum=0, Maximum=0)
    UseOwnerScale=true
    V Position=&FloatFromCurve(FloatCurve=DataInterface() {propertyValues {Curve {keys [<at 0 0.8>,<at 1 0.99>]}}}, CurveIndex=$System.NormalizedLoopAge)
   ) {module=ShapeLocation007}
   /Niagara/Modules/Update/Forces/CurlNoiseForce(Noise Frequency=0.25, Noise Strength=5, Pan Noise=true, Pan Noise Field=<0 0 -5>) {enabled=false, module=CurlNoiseForce001}
   &ApplyInitialForces() {enabled=false}
   &InitialMeshOrientation(
    Mesh Axis To Orient=<0 0 0>
    Mesh Orientation Mode=::NewEnumerator2
    Orientation Vector=<0 0 1>
    Rotation=&RandomRangeVector(Minimum=<0 -0.1 -0.1>, Maximum=<0 0.1 0.1>)
    Rotation Coordinate Space=::NewEnumerator1
    Use Rotation Vector=true
   )
   &ShapeLocation(
    Cylinder Height=2
    Cylinder Radius=&FN_LOD_Distance_Remap(Start Distance=5000, End Distance=15000, Minimum Output=6000, Maximum Output=3000, Clamp Output=true)
    MaskSpawnGroup=false
    Non Uniform Scale=<2 1 1>
    Offset=<0 0 1500>
    Offset Mode=::NewEnumerator0
    Override Seed=false
    Shape Primitive=::NewEnumerator1
    Surface Only=true
    Use Endcaps In Surface Only Mode=false
    UseOwnerScale=true
   ) {enabled=false}
  ]
  particleUpdateScript [
   &ParticleState()
   &UpdateMeshOrientation() {enabled=false, module=UpdateMeshOrientation001}
   &ScaleMeshSize(Scale Factor=&Multiply_VectorByFloat(
     Vector=<0.5 0.5 0.5>
     Float=&FN_LOD_Distance_Remap(Start Distance=5000, End Distance=50000, Minimum Output=0.5, Maximum Output=2, Clamp Output=true)
    ))
   &UpdateMeshOrientation(
    Facing Interpolation=false
    Orientation Method=::NewEnumerator1
    Override Quat=false
    TransformFacing=false
    TransformUp=false
    Up Direction:Vector3f=<0 0 -1>
   )
   &UpdateMeshOrientation(
    Rotation Rate=2
    Rotation Vector=&MakeVector(
     X=0
     Y=&Sine(Normalized Angle=$Emitter.Age, Period=&RandomRangeFloat(Minimum=2, Maximum=3), Scale=0.02, Bias=0)
     Z=0
    )
   ) {module=UpdateMeshOrientation002}
   &CurlNoiseForce(
    Noise Frequency=1
    Noise Quality / Cost=::NewEnumerator0
    Noise Strength=&FloatFromCurve(FloatCurve=DataInterface() {propertyValues {Curve {keys [<at 0 1>,<at 0.25 0>]}}})
    Pan Noise=true
    Pan Noise Field=<0 10 0>
   )
   &ScaleSpriteSize(Scale Factor=&Vector2DFromFloat(Value=&FloatFromCurve(
      FloatCurve=DataInterface() {propertyValues {Curve {keys [{arriveTangent=4.5137, interpMode=RCIM_Cubic, leaveTangent=4.5137, tangentMode=RCTM_User, time=0, value=0}, <at 1 0.5 RCIM_Cubic>]}}}
      Scale Curve=2
     )))
   &SolveForcesAndVelocity(Clamp Velocity=false, Limit Acceleration=false)
   &ScaleColor(
    Scale Alpha=&FloatFromCurve(FloatCurve=DataInterface() {propertyValues {Curve {keys [
         {arriveTangent=9.0725, interpMode=RCIM_Cubic, leaveTangent=9.0725, tangentMode=RCTM_User, time=0, value=0}
         {arriveTangent=-2.5249, interpMode=RCIM_Cubic, leaveTangent=-2.5249, tangentMode=RCTM_User, time=0.4, value=1.075}
         {arriveTangent=-0.0405, interpMode=RCIM_Cubic, leaveTangent=-0.0405, tangentMode=RCTM_User, time=1, value=0}
        ]}}})
    ScaleA=true
    ScaleRGB=true
   )
   &DynamicMaterialParameters(
    Index 0 Param 1=&Multiply_Float(
     A=25
     B=&FN_LOD_Distance_Remap(Start Distance=250, End Distance=1500, Minimum Output=2, Maximum Output=5, Clamp Output=true)
    )
    Index 0 Param 2=&Multiply_Float(
     A=1
     B=&FN_LOD_Distance_Remap(Start Distance=15000, End Distance=50000, Minimum Output=100, Maximum Output=150, Clamp Output=true)
    )
    Index 1 Param 1=50
    Module.Index0Param0Write=true
    Module.Index0Param1Write=true
    Module.Index0Param2Write=true
    Module.Index0Param3Write=true
    Module.Index1Param0Write=true
    Module.Index1Param1Write=true
    Module.Index1Param2Write=true
    Module.Index1Param3Write=true
    Write Parameter Index 1=true
   )
   &CameraOffset(Camera Offset Amount=250)
   /SetVariables(Particles.Linear Color=&MultiplyLinearColor(Linear Color=$Particles.Color, Scale Factor=<2 2 2 1>)) {isSetParameters=true, module=SetVariables_#2}
   /SetVariables(
    Particles.LightScale:Vector3f=&Multiply_Vector(
     A=/Niagara/DynamicInputs/Math/Clamp_Vector(Vector:Vector3f=$Particles.Scale, Min:Vector3f=<20 20 20>, Max:Vector3f=<25 25 25>)
     B=<0.08 0.04 0.04>
    )
    Particles.float:NiagaraFloat=&Multiply_Float(A=$Particles.CameraOffset, B=0)
   ) {isSetParameters=true, module=SetVariables_#3}
  ]
  rendererClasses [&NiagaraMeshRendererProperties]
  renderers [{rendererClass=&NiagaraMeshRendererProperties, rendererIndex=0}]
 }
 FN_Sprite001 {
  simTarget=GPUComputeSim
  emitterUpdateScript [
   &EmitterState()
   &SpawnRate(
    SpawnRate=&Multiply_Float(A=100, B=&ClampFloat(Float=$NPC.NPC_NitroDrome.NitroDrome_Emote, Min=$System.MinSpawnRate))
    Use Spawn Probability=false
   )
   &SpawnRate(
    SpawnGroup=1
    SpawnRate=&Multiply_Float(A=15, B=&ClampFloat(Float=$NPC.NPC_NitroDrome.NitroDrome_Emote, Min=$System.MinSpawnRate))
    Use Spawn Probability=false
   ) {module=SpawnRate001}
   &SpawnRate(
    Spawn Probability=2
    SpawnGroup=2
    SpawnRate=&Multiply_Float(A=15, B=&ClampFloat(Float=$NPC.NPC_NitroDrome.NitroDrome_Emote, Min=$System.MinSpawnRate))
    Use Spawn Probability=false
   ) {enabled=false, module=SpawnRate002}
   &SpawnRate(
    Spawn Probability=2
    SpawnGroup=3
    SpawnRate=&Multiply_Float(A=15, B=&ClampFloat(Float=$NPC.NPC_NitroDrome.NitroDrome_Emote, Min=$System.MinSpawnRate))
    Use Spawn Probability=false
   ) {module=SpawnRate003}
   &SpawnRate(
    Spawn Probability=2
    SpawnGroup=4
    SpawnRate=&Multiply_Float(A=15, B=&ClampFloat(Float=$NPC.NPC_NitroDrome.NitroDrome_Emote, Min=$System.MinSpawnRate))
    Use Spawn Probability=false
   ) {module=SpawnRate004}
   &SpawnRate(
    Spawn Probability=2
    SpawnGroup=5
    SpawnRate=&Multiply_Float(A=15, B=&ClampFloat(Float=$NPC.NPC_NitroDrome.NitroDrome_Emote, Min=$System.MinSpawnRate))
    Use Spawn Probability=false
   ) {enabled=false, module=SpawnRate005}
   &SpawnRate(
    Spawn Probability=2
    SpawnGroup=6
    SpawnRate=&Multiply_Float(A=15, B=&ClampFloat(Float=$NPC.NPC_NitroDrome.NitroDrome_Emote, Min=$System.MinSpawnRate))
    Use Spawn Probability=false
   ) {module=SpawnRate006}
  ]
  particleSpawnScript [
   &InitializeParticle(
    Color=&ColorFromCurve(
     ColorCurve=DataInterface() {propertyValues {
       AlphaCurve {keys [<at 0 1 RCIM_Cubic>,<at 0.9977 1 RCIM_Cubic>]}
       BlueCurve {keys [<at 0.2936 0>,<at 0.608 0>]}
       GreenCurve {keys [<at 0.2936 0.1016>,<at 0.608 0.2592>]}
       RedCurve {keys [<at 0.2936 1>,<at 0.608 1>]}
      }}
     CurveIndex=&RandomRangeFloat(Minimum=0, Maximum=1)
    )
    Lifetime Max=1
    Lifetime Mode=::NewEnumerator1
    Mesh Scale Mode=::NewEnumerator1
    Mesh Uniform Scale Max=250
    Mesh Uniform Scale Min=50
    Sprite Size Mode=::NewEnumerator1
    Uniform Sprite Size Max=300
    Uniform Sprite Size Min=250
    UsePositionOffset=false
    Write Material Random=true
   ) {module=InitializeParticle001}
   &AddVelocityFromPoint(
    Module.Use Falloff Distance=false
    Origin Offset=&UniformRangedVector(Minimum=<0 0 -100>, Maximum=<0 0 -100>)
    Origin Offset Coordinate Space=::Simulation
    Velocity Strength=&UniformRangedFloat(Minimum=10, Maximum=15)
   ) {enabled=false}
   &ShapeLocation(
    Handle Radius=0
    Large Radius=1000
    MaskSpawnGroup=true
    Non Uniform Scale=<2.5 1.5 0>
    Shape Primitive=::NewEnumerator3
    Torus Distribution Mode:ENiagaraTorusDistributionMode=::NewEnumerator1
    UseOwnerScale=true
   ) {module=ShapeLocation001}
   &ShapeLocation(
    MaskSpawnGroup=true
    Non Uniform Scale=<1.3 1 0.5>
    Offset=<0 -500 -2400>
    Offset Mode=::NewEnumerator0
    Rotation Angle=0.325
    Rotation Axis=<0 0 1>
    Rotation Coordinate Space=::Simulation
    Rotation Mode=::NewEnumerator1
    Spawn Group Mask=1
    Sphere Distribution=::NewEnumerator1
    Sphere Radius=5000
    U Position=&RandomRangeFloat(Minimum=0, Maximum=0)
    UseOwnerScale=true
    V Position=&FloatFromCurve(FloatCurve=DataInterface() {propertyValues {Curve {keys [<at 0 0.8>,<at 1 0.99>]}}}, CurveIndex=$System.NormalizedLoopAge)
   ) {module=ShapeLocation002}
   &ShapeLocation(
    MaskSpawnGroup=true
    Non Uniform Scale=<1.6 1 0.5>
    Offset=<0 250 -2400>
    Offset Mode=::NewEnumerator0
    Rotation Angle=0.507
    Rotation Axis=<0 0 1>
    Rotation Coordinate Space=::Simulation
    Rotation Mode=::NewEnumerator1
    Spawn Group Mask=2
    Sphere Distribution=::NewEnumerator1
    Sphere Radius=5000
    U Position=&RandomRangeFloat(Minimum=0, Maximum=0)
    UseOwnerScale=true
    V Position=&FloatFromCurve(FloatCurve=DataInterface() {propertyValues {Curve {keys [<at 0 0.7>,<at 1 0.99>]}}}, CurveIndex=$System.NormalizedLoopAge)
   ) {module=ShapeLocation003}
   &ShapeLocation(
    MaskSpawnGroup=true
    Non Uniform Scale=<1.2 1 0.5>
    Offset=<0 250 -2400>
    Offset Mode=::NewEnumerator0
    Rotation Angle=0.67
    Rotation Axis=<0 0 1>
    Rotation Coordinate Space=::Simulation
    Rotation Mode=::NewEnumerator1
    Spawn Group Mask=3
    Sphere Distribution=::NewEnumerator1
    Sphere Radius=5000
    U Position=&RandomRangeFloat(Minimum=0, Maximum=0)
    UseOwnerScale=true
    V Position=&FloatFromCurve(FloatCurve=DataInterface() {propertyValues {Curve {keys [<at 0 0.7>,<at 1 0.99>]}}}, CurveIndex=$System.NormalizedLoopAge)
   ) {module=ShapeLocation004}
   &ShapeLocation(
    MaskSpawnGroup=true
    Non Uniform Scale=<1.2 1 0.5>
    Offset=<0 250 -2400>
    Offset Mode=::NewEnumerator0
    Rotation Angle=0.82
    Rotation Axis=<0 0 1>
    Rotation Coordinate Space=::Simulation
    Rotation Mode=::NewEnumerator1
    Spawn Group Mask=4
    Sphere Distribution=::NewEnumerator1
    Sphere Radius=5000
    U Position=&RandomRangeFloat(Minimum=0, Maximum=0)
    UseOwnerScale=true
    V Position=&FloatFromCurve(FloatCurve=DataInterface() {propertyValues {Curve {keys [<at 0 0.5>,<at 1 0.99>]}}}, CurveIndex=$System.NormalizedLoopAge)
   ) {module=ShapeLocation005}
   &ShapeLocation(
    MaskSpawnGroup=true
    Non Uniform Scale=<1.6 1 0.5>
    Offset=<0 250 -2500>
    Offset Mode=::NewEnumerator0
    Rotation Angle=0.995
    Rotation Axis=<0 0 1>
    Rotation Coordinate Space=::Simulation
    Rotation Mode=::NewEnumerator1
    Spawn Group Mask=5
    Sphere Distribution=::NewEnumerator1
    Sphere Radius=5000
    U Position=&RandomRangeFloat(Minimum=0, Maximum=0)
    UseOwnerScale=true
    V Position=&FloatFromCurve(FloatCurve=DataInterface() {propertyValues {Curve {keys [<at 0 0.5>,<at 1 0.99>]}}}, CurveIndex=$System.NormalizedLoopAge)
   ) {module=ShapeLocation006}
   &ShapeLocation(
    MaskSpawnGroup=true
    Non Uniform Scale=<1.3 1 0.5>
    Offset=<0 -500 -2500>
    Offset Mode=::NewEnumerator0
    Rotation Angle=0.18
    Rotation Axis=<0 0 1>
    Rotation Coordinate Space=::Simulation
    Rotation Mode=::NewEnumerator1
    Spawn Group Mask=6
    Sphere Distribution=::NewEnumerator1
    Sphere Radius=5000
    U Position=&RandomRangeFloat(Minimum=0, Maximum=0)
    UseOwnerScale=true
    V Position=&FloatFromCurve(FloatCurve=DataInterface() {propertyValues {Curve {keys [<at 0 0.8>,<at 1 0.99>]}}}, CurveIndex=$System.NormalizedLoopAge)
   ) {module=ShapeLocation007}
   &CurlNoiseForce(Noise Frequency=0.25, Noise Strength=5, Pan Noise=true, Pan Noise Field=<0 0 -5>) {enabled=false, module=CurlNoiseForce001}
   &ApplyInitialForces() {enabled=false}
   &InitialMeshOrientation(
    Mesh Axis To Orient=<0 0 0>
    Mesh Orientation Mode=::NewEnumerator2
    Orientation Vector=<0 0 1>
    Rotation=&RandomRangeVector(Minimum=<0 0 0>, Maximum=<0 0 0>)
    Rotation Coordinate Space=::NewEnumerator1
    Use Rotation Vector=true
   )
   &ShapeLocation(
    Cylinder Height=2
    Cylinder Radius=&FN_LOD_Distance_Remap(Start Distance=5000, End Distance=15000, Minimum Output=6000, Maximum Output=3000, Clamp Output=true)
    MaskSpawnGroup=false
    Non Uniform Scale=<2 1 1>
    Offset=<0 0 1500>
    Offset Mode=::NewEnumerator0
    Override Seed=false
    Shape Primitive=::NewEnumerator1
    Surface Only=true
    Use Endcaps In Surface Only Mode=false
    UseOwnerScale=true
   ) {enabled=false}
  ]
  particleUpdateScript [
   &ParticleState()
   &UpdateMeshOrientation() {enabled=false, module=UpdateMeshOrientation001}
   &ScaleMeshSize(Scale Factor=&Multiply_VectorByFloat(
     Vector=<0.25 0.25 0.25>
     Float=&FN_LOD_Distance_Remap(Start Distance=5000, End Distance=50000, Minimum Output=0.5, Maximum Output=2, Clamp Output=true)
    ))
   &UpdateMeshOrientation(
    Facing Interpolation=false
    Orientation Method=::NewEnumerator1
    Override Quat=false
    TransformFacing=false
    TransformUp=false
    Up Direction:Vector3f=<0 0 -1>
   )
   &UpdateMeshOrientation(
    Rotation Rate=2
    Rotation Vector=&MakeVector(
     X=0
     Y=&Sine(Normalized Angle=$Emitter.Age, Period=&RandomRangeFloat(Minimum=2, Maximum=3), Scale=0.02, Bias=0)
     Z=0
    )
   ) {module=UpdateMeshOrientation002}
   &CurlNoiseForce(
    Noise Frequency=1
    Noise Quality / Cost=::NewEnumerator0
    Noise Strength=&FloatFromCurve(FloatCurve=DataInterface() {propertyValues {Curve {keys [<at 0 1>,<at 0.25 0>]}}})
    Pan Noise=true
    Pan Noise Field=<0 10 0>
   )
   &ScaleSpriteSize(Scale Factor=&Vector2DFromFloat(Value=&FloatFromCurve(
      FloatCurve=DataInterface() {propertyValues {Curve {keys [{arriveTangent=4.5137, interpMode=RCIM_Cubic, leaveTangent=4.5137, tangentMode=RCTM_User, time=0, value=0}, <at 1 0.5 RCIM_Cubic>]}}}
      Scale Curve=2
     )))
   &SolveForcesAndVelocity(Clamp Velocity=false, Limit Acceleration=false)
   &ScaleColor(
    Scale Alpha=&FloatFromCurve(
     FloatCurve=DataInterface() {propertyValues {Curve {keys [
         {arriveTangent=9.0725, interpMode=RCIM_Cubic, leaveTangent=9.0725, tangentMode=RCTM_User, time=0, value=0}
         {arriveTangent=-2.5249, interpMode=RCIM_Cubic, leaveTangent=-2.5249, tangentMode=RCTM_User, time=0.4, value=1.075}
         {arriveTangent=-0.0405, interpMode=RCIM_Cubic, leaveTangent=-0.0405, tangentMode=RCTM_User, time=1, value=0}
        ]}}}
     Scale Curve=0.5
    )
    ScaleA=true
    ScaleRGB=true
   )
   &DynamicMaterialParameters(
    Index 0 Param 1=&Multiply_Float(
     A=25
     B=&FN_LOD_Distance_Remap(Start Distance=250, End Distance=1500, Minimum Output=2, Maximum Output=5, Clamp Output=true)
    )
    Index 0 Param 2=&Multiply_Float(
     A=1
     B=&FN_LOD_Distance_Remap(Start Distance=15000, End Distance=50000, Minimum Output=5, Maximum Output=2500, Clamp Output=true)
    )
    Index 1 Param 1=50
    Module.Index0Param0Write=true
    Module.Index0Param1Write=true
    Module.Index0Param2Write=true
    Module.Index0Param3Write=true
    Module.Index1Param0Write=true
    Module.Index1Param1Write=true
    Module.Index1Param2Write=true
    Module.Index1Param3Write=true
    Write Parameter Index 1=true
   )
   &CameraOffset(Camera Offset Amount=250)
   /SetVariables(Particles.Linear Color=&MultiplyLinearColor(Linear Color=$Particles.Color, Scale Factor=<2 2 2 1>)) {isSetParameters=true, module=SetVariables_#2}
   /SetVariables(
    Particles.LightScale:Vector3f=&Multiply_Vector(
     A=&Clamp_Vector(Vector:Vector3f=$Particles.Scale, Min:Vector3f=<20 20 20>, Max:Vector3f=<25 25 25>)
     B=<0.08 0.04 0.04>
    )
    Particles.float:NiagaraFloat=&Multiply_Float(A=$Particles.CameraOffset, B=0)
   ) {isSetParameters=true, module=SetVariables_#3}
  ]
  rendererClasses [&NiagaraMeshRendererProperties]
  renderers [{rendererClass=&NiagaraMeshRendererProperties, rendererIndex=0}]
 }
 FN_Sprite002 {
  simTarget=GPUComputeSim
  emitterUpdateScript [
   &EmitterState()
   &SpawnRate(
    SpawnRate=&Multiply_Float(A=100, B=&ClampFloat(Float=$NPC.NPC_NitroDrome.NitroDrome_Emote, Min=$System.MinSpawnRate))
    Use Spawn Probability=false
   )
   &SpawnRate(
    SpawnGroup=1
    SpawnRate=&Multiply_Float(A=15, B=&ClampFloat(Float=$NPC.NPC_NitroDrome.NitroDrome_Emote, Min=$System.MinSpawnRate))
    Use Spawn Probability=false
   ) {module=SpawnRate001}
   &SpawnRate(
    Spawn Probability=2
    SpawnGroup=2
    SpawnRate=&Multiply_Float(A=15, B=&ClampFloat(Float=$NPC.NPC_NitroDrome.NitroDrome_Emote, Min=$System.MinSpawnRate))
    Use Spawn Probability=false
   ) {enabled=false, module=SpawnRate002}
   &SpawnRate(
    Spawn Probability=2
    SpawnGroup=3
    SpawnRate=&Multiply_Float(A=15, B=&ClampFloat(Float=$NPC.NPC_NitroDrome.NitroDrome_Emote, Min=$System.MinSpawnRate))
    Use Spawn Probability=false
   ) {module=SpawnRate003}
   &SpawnRate(
    Spawn Probability=2
    SpawnGroup=4
    SpawnRate=&Multiply_Float(A=15, B=&ClampFloat(Float=$NPC.NPC_NitroDrome.NitroDrome_Emote, Min=$System.MinSpawnRate))
    Use Spawn Probability=false
   ) {module=SpawnRate004}
   &SpawnRate(
    Spawn Probability=2
    SpawnGroup=5
    SpawnRate=&Multiply_Float(A=15, B=&ClampFloat(Float=$NPC.NPC_NitroDrome.NitroDrome_Emote, Min=$System.MinSpawnRate))
    Use Spawn Probability=false
   ) {enabled=false, module=SpawnRate005}
   &SpawnRate(
    Spawn Probability=2
    SpawnGroup=6
    SpawnRate=&Multiply_Float(A=15, B=&ClampFloat(Float=$NPC.NPC_NitroDrome.NitroDrome_Emote, Min=$System.MinSpawnRate))
    Use Spawn Probability=false
   ) {module=SpawnRate006}
  ]
  particleSpawnScript [
   &InitializeParticle(
    Color=&ColorFromCurve(
     ColorCurve=DataInterface() {propertyValues {
       AlphaCurve {keys [<at 0 1 RCIM_Cubic>,<at 0.9977 1 RCIM_Cubic>]}
       BlueCurve {keys [<at 0.2936 0>,<at 0.608 0>]}
       GreenCurve {keys [<at 0.2936 0.1016>,<at 0.608 0.2592>]}
       RedCurve {keys [<at 0.2936 1>,<at 0.608 1>]}
      }}
     CurveIndex=&RandomRangeFloat(Minimum=0, Maximum=1)
    )
    Lifetime Max=1
    Lifetime Mode=::NewEnumerator1
    Mesh Scale Mode=::NewEnumerator1
    Mesh Uniform Scale Max=250
    Mesh Uniform Scale Min=50
    Sprite Size Mode=::NewEnumerator1
    Uniform Sprite Size Max=300
    Uniform Sprite Size Min=250
    UsePositionOffset=false
    Write Material Random=true
   ) {module=InitializeParticle001}
   &AddVelocityFromPoint(
    Module.Use Falloff Distance=false
    Origin Offset=&UniformRangedVector(Minimum=<0 0 -100>, Maximum=<0 0 -100>)
    Origin Offset Coordinate Space=::Simulation
    Velocity Strength=&UniformRangedFloat(Minimum=10, Maximum=15)
   ) {enabled=false}
   &ShapeLocation(
    Handle Radius=0
    Large Radius=1000
    MaskSpawnGroup=true
    Non Uniform Scale=<2.5 1.5 0>
    Shape Primitive=::NewEnumerator3
    Torus Distribution Mode:ENiagaraTorusDistributionMode=::NewEnumerator1
    UseOwnerScale=true
   ) {module=ShapeLocation001}
   &ShapeLocation(
    MaskSpawnGroup=true
    Non Uniform Scale=<1.3 1 0.5>
    Offset=<0 -500 -2400>
    Offset Mode=::NewEnumerator0
    Rotation Angle=0.325
    Rotation Axis=<0 0 1>
    Rotation Coordinate Space=::Simulation
    Rotation Mode=::NewEnumerator1
    Spawn Group Mask=1
    Sphere Distribution=::NewEnumerator1
    Sphere Radius=5000
    U Position=&RandomRangeFloat(Minimum=0, Maximum=0)
    UseOwnerScale=true
    V Position=&FloatFromCurve(FloatCurve=DataInterface() {propertyValues {Curve {keys [<at 0 0.8>,<at 1 0.99>]}}}, CurveIndex=$System.NormalizedLoopAge)
   ) {module=ShapeLocation002}
   &ShapeLocation(
    MaskSpawnGroup=true
    Non Uniform Scale=<1.6 1 0.5>
    Offset=<0 250 -2400>
    Offset Mode=::NewEnumerator0
    Rotation Angle=0.507
    Rotation Axis=<0 0 1>
    Rotation Coordinate Space=::Simulation
    Rotation Mode=::NewEnumerator1
    Spawn Group Mask=2
    Sphere Distribution=::NewEnumerator1
    Sphere Radius=5000
    U Position=&RandomRangeFloat(Minimum=0, Maximum=0)
    UseOwnerScale=true
    V Position=&FloatFromCurve(FloatCurve=DataInterface() {propertyValues {Curve {keys [<at 0 0.7>,<at 1 0.99>]}}}, CurveIndex=$System.NormalizedLoopAge)
   ) {module=ShapeLocation003}
   &ShapeLocation(
    MaskSpawnGroup=true
    Non Uniform Scale=<1.2 1 0.5>
    Offset=<0 250 -2400>
    Offset Mode=::NewEnumerator0
    Rotation Angle=0.67
    Rotation Axis=<0 0 1>
    Rotation Coordinate Space=::Simulation
    Rotation Mode=::NewEnumerator1
    Spawn Group Mask=3
    Sphere Distribution=::NewEnumerator1
    Sphere Radius=5000
    U Position=&RandomRangeFloat(Minimum=0, Maximum=0)
    UseOwnerScale=true
    V Position=&FloatFromCurve(FloatCurve=DataInterface() {propertyValues {Curve {keys [<at 0 0.7>,<at 1 0.99>]}}}, CurveIndex=$System.NormalizedLoopAge)
   ) {module=ShapeLocation004}
   &ShapeLocation(
    MaskSpawnGroup=true
    Non Uniform Scale=<1.2 1 0.5>
    Offset=<0 250 -2400>
    Offset Mode=::NewEnumerator0
    Rotation Angle=0.82
    Rotation Axis=<0 0 1>
    Rotation Coordinate Space=::Simulation
    Rotation Mode=::NewEnumerator1
    Spawn Group Mask=4
    Sphere Distribution=::NewEnumerator1
    Sphere Radius=5000
    U Position=&RandomRangeFloat(Minimum=0, Maximum=0)
    UseOwnerScale=true
    V Position=&FloatFromCurve(FloatCurve=DataInterface() {propertyValues {Curve {keys [<at 0 0.5>,<at 1 0.99>]}}}, CurveIndex=$System.NormalizedLoopAge)
   ) {module=ShapeLocation005}
   &ShapeLocation(
    MaskSpawnGroup=true
    Non Uniform Scale=<1.6 1 0.5>
    Offset=<0 250 -2500>
    Offset Mode=::NewEnumerator0
    Rotation Angle=0.995
    Rotation Axis=<0 0 1>
    Rotation Coordinate Space=::Simulation
    Rotation Mode=::NewEnumerator1
    Spawn Group Mask=5
    Sphere Distribution=::NewEnumerator1
    Sphere Radius=5000
    U Position=&RandomRangeFloat(Minimum=0, Maximum=0)
    UseOwnerScale=true
    V Position=&FloatFromCurve(FloatCurve=DataInterface() {propertyValues {Curve {keys [<at 0 0.5>,<at 1 0.99>]}}}, CurveIndex=$System.NormalizedLoopAge)
   ) {module=ShapeLocation006}
   &ShapeLocation(
    MaskSpawnGroup=true
    Non Uniform Scale=<1.3 1 0.5>
    Offset=<0 -500 -2500>
    Offset Mode=::NewEnumerator0
    Rotation Angle=0.18
    Rotation Axis=<0 0 1>
    Rotation Coordinate Space=::Simulation
    Rotation Mode=::NewEnumerator1
    Spawn Group Mask=6
    Sphere Distribution=::NewEnumerator1
    Sphere Radius=5000
    U Position=&RandomRangeFloat(Minimum=0, Maximum=0)
    UseOwnerScale=true
    V Position=&FloatFromCurve(FloatCurve=DataInterface() {propertyValues {Curve {keys [<at 0 0.8>,<at 1 0.99>]}}}, CurveIndex=$System.NormalizedLoopAge)
   ) {module=ShapeLocation007}
   &CurlNoiseForce(Noise Frequency=0.25, Noise Strength=5, Pan Noise=true, Pan Noise Field=<0 0 -5>) {enabled=false, module=CurlNoiseForce001}
   &ApplyInitialForces() {enabled=false}
   &InitialMeshOrientation(
    Mesh Axis To Orient=<0 0 0>
    Mesh Orientation Mode=::NewEnumerator2
    Orientation Vector=<0 0 1>
    Rotation=&RandomRangeVector(Minimum=<0 -0.1 -0.1>, Maximum=<0 0.1 0.1>)
    Rotation Coordinate Space=::NewEnumerator1
    Use Rotation Vector=true
   )
   &ShapeLocation(
    Cylinder Height=2
    Cylinder Radius=&FN_LOD_Distance_Remap(Start Distance=5000, End Distance=15000, Minimum Output=6000, Maximum Output=3000, Clamp Output=true)
    MaskSpawnGroup=false
    Non Uniform Scale=<2 1 1>
    Offset=<0 0 1500>
    Offset Mode=::NewEnumerator0
    Override Seed=false
    Shape Primitive=::NewEnumerator1
    Surface Only=true
    Use Endcaps In Surface Only Mode=false
    UseOwnerScale=true
   ) {enabled=false}
  ]
  particleUpdateScript [
   &ParticleState()
   &UpdateMeshOrientation() {enabled=false, module=UpdateMeshOrientation001}
   &ScaleMeshSize(Scale Factor=&Multiply_VectorByFloat(
     Vector=<0.25 0.25 0.25>
     Float=&FN_LOD_Distance_Remap(Start Distance=5000, End Distance=50000, Minimum Output=0.5, Maximum Output=2, Clamp Output=true)
    ))
   &UpdateMeshOrientation(
    Facing Interpolation=false
    Orientation Method=::NewEnumerator1
    Override Quat=false
    TransformFacing=false
    TransformUp=false
    Up Direction:Vector3f=<0 0 -1>
   )
   &UpdateMeshOrientation(
    Rotation Rate=2
    Rotation Vector=&MakeVector(
     X=0
     Y=&Sine(Normalized Angle=$Emitter.Age, Period=&RandomRangeFloat(Minimum=2, Maximum=3), Scale=0.02, Bias=0)
     Z=0
    )
   ) {module=UpdateMeshOrientation002}
   &CurlNoiseForce(
    Noise Frequency=1
    Noise Quality / Cost=::NewEnumerator0
    Noise Strength=&FloatFromCurve(FloatCurve=DataInterface() {propertyValues {Curve {keys [<at 0 1>,<at 0.25 0>]}}})
    Pan Noise=true
    Pan Noise Field=<0 10 0>
   )
   &ScaleSpriteSize(Scale Factor=&Vector2DFromFloat(Value=&FloatFromCurve(
      FloatCurve=DataInterface() {propertyValues {Curve {keys [{arriveTangent=4.5137, interpMode=RCIM_Cubic, leaveTangent=4.5137, tangentMode=RCTM_User, time=0, value=0}, <at 1 0.5 RCIM_Cubic>]}}}
      Scale Curve=2
     )))
   &SolveForcesAndVelocity(Clamp Velocity=false, Limit Acceleration=false)
   &ScaleColor(
    Scale Alpha=&FloatFromCurve(
     FloatCurve=DataInterface() {propertyValues {Curve {keys [
         {arriveTangent=9.0725, interpMode=RCIM_Cubic, leaveTangent=9.0725, tangentMode=RCTM_User, time=0, value=0}
         {arriveTangent=-2.5249, interpMode=RCIM_Cubic, leaveTangent=-2.5249, tangentMode=RCTM_User, time=0.4, value=1.075}
         {arriveTangent=-0.0405, interpMode=RCIM_Cubic, leaveTangent=-0.0405, tangentMode=RCTM_User, time=1, value=0}
        ]}}}
     Scale Curve=0.5
    )
    ScaleA=true
    ScaleRGB=true
   )
   &DynamicMaterialParameters(
    Index 0 Param 1=&Multiply_Float(
     A=25
     B=&FN_LOD_Distance_Remap(Start Distance=250, End Distance=1500, Minimum Output=2, Maximum Output=5, Clamp Output=true)
    )
    Index 0 Param 2=&Multiply_Float(
     A=1
     B=&FN_LOD_Distance_Remap(Start Distance=15000, End Distance=50000, Minimum Output=5, Maximum Output=2500, Clamp Output=true)
    )
    Index 1 Param 1=50
    Module.Index0Param0Write=true
    Module.Index0Param1Write=true
    Module.Index0Param2Write=true
    Module.Index0Param3Write=true
    Module.Index1Param0Write=true
    Module.Index1Param1Write=true
    Module.Index1Param2Write=true
    Module.Index1Param3Write=true
    Write Parameter Index 1=true
   )
   &CameraOffset(Camera Offset Amount=250)
   /SetVariables(Particles.Linear Color=&MultiplyLinearColor(Linear Color=$Particles.Color, Scale Factor=<2 2 2 1>)) {isSetParameters=true, module=SetVariables_#2}
   /SetVariables(
    Particles.LightScale:Vector3f=&Multiply_Vector(
     A=&Clamp_Vector(Vector:Vector3f=$Particles.Scale, Min:Vector3f=<20 20 20>, Max:Vector3f=<25 25 25>)
     B=<0.08 0.04 0.04>
    )
    Particles.float:NiagaraFloat=&Multiply_Float(A=$Particles.CameraOffset, B=0)
   ) {isSetParameters=true, module=SetVariables_#3}
  ]
  rendererClasses [&NiagaraSpriteRendererProperties]
  renderers [{rendererClass=&NiagaraSpriteRendererProperties}]
 }
 Main002 {
  bEnabled=false
  simTarget=GPUComputeSim
  emitterUpdateScript [
   &EmitterState()
   &SpawnRate(
    SpawnRate=&Multiply_Float(A=25, B=&ClampFloat(Float=$NPC.NPC_NitroDrome.NitroDrome_VehicleEntered, Min=1))
    Use Spawn Probability=false
   )
   &SpawnRate(
    SpawnGroup=1
    SpawnRate=&Multiply_Float(A=15, B=&ClampFloat(Float=$NPC.NPC_NitroDrome.NitroDrome_VehicleEntered, Min=1))
    Use Spawn Probability=false
   ) {module=SpawnRate001}
   &SpawnRate(
    Spawn Probability=2
    SpawnGroup=2
    SpawnRate=&Multiply_Float(A=15, B=&ClampFloat(Float=$NPC.NPC_NitroDrome.NitroDrome_VehicleEntered, Min=1))
    Use Spawn Probability=false
   ) {module=SpawnRate002}
   &SpawnRate(
    Spawn Probability=2
    SpawnGroup=3
    SpawnRate=&Multiply_Float(A=15, B=&ClampFloat(Float=$NPC.NPC_NitroDrome.NitroDrome_VehicleEntered, Min=1))
    Use Spawn Probability=false
   ) {module=SpawnRate003}
   &SpawnRate(
    Spawn Probability=2
    SpawnGroup=4
    SpawnRate=&Multiply_Float(A=15, B=&ClampFloat(Float=$NPC.NPC_NitroDrome.NitroDrome_VehicleEntered, Min=1))
    Use Spawn Probability=false
   ) {module=SpawnRate004}
   &SpawnRate(
    Spawn Probability=2
    SpawnGroup=5
    SpawnRate=&Multiply_Float(A=15, B=&ClampFloat(Float=$NPC.NPC_NitroDrome.NitroDrome_VehicleEntered, Min=1))
    Use Spawn Probability=false
   ) {module=SpawnRate005}
   &SpawnRate(
    Spawn Probability=2
    SpawnGroup=6
    SpawnRate=&Multiply_Float(A=15, B=&ClampFloat(Float=$NPC.NPC_NitroDrome.NitroDrome_VehicleEntered, Min=1))
    Use Spawn Probability=false
   ) {module=SpawnRate006}
  ]
  particleSpawnScript [
   &InitializeParticle(
    Color=&SelectLinearColorFromArray(
     Array Sampling Mode=::NewEnumerator1
     Direct Array Index=&SpawnGroup()
     Color Selection Array=DataInterface() {propertyValues {ColorData [<rgba 1 0 0 1>, <rgba 0.0673 1 0 1>, <rgba 0.0129 0 1 1>, <rgba 1 0 0.7954 1>, <rgba 1 0.7858 0 1>, <rgba 0 0.9786 1 1>, <rgba 0 1 0.0532 1>]}}
     Random Seed=?
    )
    Sprite Size Mode=::NewEnumerator3
    Uniform Sprite Size=250
    UsePositionOffset=false
    Write Material Random=true
   )
   &AddVelocity(
    Cone Angle=25
    Cone Axis=<0 0 -1>
    Inner Cone Angle=20
    Module.UseConeFalloff=false
    Module.UseVelDistribution=false
    Override Seed=false
    Velocity Mode=::NewEnumerator2
    Velocity Speed=&RandomRangeFloat(Minimum=50, Maximum=250)
   ) {enabled=false}
   &ShapeLocation(
    Handle Radius=0
    Large Radius=1000
    MaskSpawnGroup=true
    Non Uniform Scale=<2 1 0>
    Shape Primitive=::NewEnumerator3
    Torus Distribution Mode:ENiagaraTorusDistributionMode=::NewEnumerator1
    UseOwnerScale=true
   ) {module=ShapeLocation001}
   &ShapeLocation(
    Hemisphere Distribution=<0 0.5>
    MaskSpawnGroup=true
    Non Uniform Scale=<0.5 1 1>
    Offset=<2500 0 -2500>
    Offset Mode=::NewEnumerator0
    Override Seed=false
    Rotation Angle=0.5
    Rotation Mode=::NewEnumerator1
    Spawn Group Mask=1
    Sphere Radius=5000
    Sphere Surface Distribution=0.9286
    UseOwnerScale=true
   )
   &ShapeLocation(
    Hemisphere Distribution=<0 0.5>
    MaskSpawnGroup=true
    Non Uniform Scale=<0.8 1 -1.3>
    Offset=<-2500 0 -4000>
    Offset Mode=::NewEnumerator0
    Override Seed=false
    Rotation Angle=0.5
    Rotation Mode=::NewEnumerator1
    Spawn Group Mask=2
    Sphere Radius=5000
    Sphere Surface Distribution=1
    UseOwnerScale=true
   ) {module=ShapeLocation003}
   &ShapeLocation(
    Hemisphere Distribution=<0 0.3>
    MaskSpawnGroup=true
    Non Uniform Scale=<1.2 1 0.5>
    Offset=<0 250 -2500>
    Offset Mode=::NewEnumerator0
    Override Seed=false
    Rotation Angle=0.67
    Rotation Axis=<0 0 1>
    Rotation Coordinate Space=::Simulation
    Rotation Mode=::NewEnumerator1
    Spawn Group Mask=3
    Sphere Radius=5000
    Sphere Surface Distribution=1
    UseOwnerScale=true
   ) {module=ShapeLocation002}
   &ShapeLocation(
    MaskSpawnGroup=true
    Non Uniform Scale=<1.2 1 0.5>
    Offset=<0 250 -2400>
    Offset Mode=::NewEnumerator0
    Rotation Angle=0.82
    Rotation Axis=<0 0 1>
    Rotation Coordinate Space=::Simulation
    Rotation Mode=::NewEnumerator1
    Spawn Group Mask=4
    Sphere Distribution=::NewEnumerator1
    Sphere Radius=5000
    U Position=&RandomRangeFloat(Minimum=0, Maximum=0)
    UseOwnerScale=true
    V Position=&FloatFromCurve(FloatCurve=DataInterface() {propertyValues {Curve {keys [<at 0 0.5>,<at 1 0.99>]}}}, CurveIndex=$System.NormalizedLoopAge)
   ) {module=ShapeLocation004}
   &ShapeLocation(
    MaskSpawnGroup=true
    Non Uniform Scale=<1.6 1 0.5>
    Offset=<0 250 -2500>
    Offset Mode=::NewEnumerator0
    Rotation Angle=0.995
    Rotation Axis=<0 0 1>
    Rotation Coordinate Space=::Simulation
    Rotation Mode=::NewEnumerator1
    Spawn Group Mask=5
    Sphere Distribution=::NewEnumerator1
    Sphere Radius=5000
    U Position=&RandomRangeFloat(Minimum=0, Maximum=0)
    UseOwnerScale=true
    V Position=&FloatFromCurve(FloatCurve=DataInterface() {propertyValues {Curve {keys [<at 0 0.5>,<at 1 0.99>]}}}, CurveIndex=$System.NormalizedLoopAge)
   ) {module=ShapeLocation005}
   &ShapeLocation(
    MaskSpawnGroup=true
    Non Uniform Scale=<1.3 1 0.5>
    Offset=<0 -500 -2500>
    Offset Mode=::NewEnumerator0
    Rotation Angle=0.18
    Rotation Axis=<0 0 1>
    Rotation Coordinate Space=::Simulation
    Rotation Mode=::NewEnumerator1
    Spawn Group Mask=6
    Sphere Distribution=::NewEnumerator1
    Sphere Radius=5000
    U Position=&RandomRangeFloat(Minimum=0, Maximum=0)
    UseOwnerScale=true
    V Position=&FloatFromCurve(FloatCurve=DataInterface() {propertyValues {Curve {keys [<at 0 0.8>,<at 1 0.99>]}}}, CurveIndex=$System.NormalizedLoopAge)
   ) {module=ShapeLocation006}
  ]
  particleUpdateScript [
   &GravityForce(Gravity=<0 0 -1600>) {enabled=false}
   &Drag(Drag=0, Use Linear Drag=true, Use Rotational Drag=false)
   /Niagara/Modules/Update/Forces/V2/CurlNoiseForce(Noise Frequency=5, Noise Strength=2500, Pan Noise=false) {enabled=false}
   &SolveForcesAndVelocity(Clamp Velocity=false, Limit Acceleration=false)
   &ParticleState()
   &DynamicMaterialParameters(
    Index 0 Param 1=0.8
    Index 0 Param 2=1
    Index 0 Param 3=1
    Index 1 Param 1=150
    Index 1 Param 4=1
    Module.Param0WriteEnabled=true
    Module.Param0WriteEnabled1=true
    Module.Param1WriteEnabled=true
    Module.Param1WriteEnabled1=true
    Module.Param2WriteEnabled=true
    Module.Param2WriteEnabled1=true
    Module.Param3WriteEnabled=false
    Module.Param3WriteEnabled1=true
    Write Parameter Index 1=true
   ) {enabled=false}
   &ScaleColor(
    Scale Alpha=&FloatFromCurve(
     FloatCurve=DataInterface() {propertyValues {Curve {keys [
         {arriveTangent=9.0725, interpMode=RCIM_Cubic, leaveTangent=9.0725, tangentMode=RCTM_User, time=0, value=0}
         {arriveTangent=-7.2695, interpMode=RCIM_Cubic, leaveTangent=-7.2695, tangentMode=RCTM_User, time=0.2, value=1.75}
         <at 0.8 1.8>
         {arriveTangent=-0.0405, interpMode=RCIM_Cubic, leaveTangent=-0.0405, tangentMode=RCTM_User, time=1, value=0}
        ]}}}
     Scale Curve=0.02
    )
    ScaleA=true
    ScaleRGB=true
   ) {enabled=false, module=ScaleColor001}
   &ScaleSpriteSize(Uniform Curve Sprite Scale=DataInterface() {propertyValues {Curve {keys [{interpMode=RCIM_Cubic, tangentMode=RCTM_User, time=0, value=1}, {interpMode=RCIM_Cubic, tangentMode=RCTM_User, time=1, value=0}]}}}) {enabled=false, module=ScaleSpriteSize001}
  ]
  rendererClasses [&NiagaraSpriteRendererProperties]
  renderers [{rendererClass=&NiagaraSpriteRendererProperties}]
 }
}
summary {
 systemName=NS_NitroDrome_Reaction_Emote
 emitters [
  {emitterName=FN_Sprite, simTarget=GPUComputeSim, rendererClasses [&NiagaraMeshRendererProperties]}
  {bEnabled=false, emitterName=Main002, simTarget=GPUComputeSim, rendererClasses [&NiagaraSpriteRendererProperties]}
  {emitterName=FN_Sprite001, simTarget=GPUComputeSim, rendererClasses [&NiagaraMeshRendererProperties]}
  {emitterName=FN_Sprite002, simTarget=GPUComputeSim, rendererClasses [&NiagaraSpriteRendererProperties]}
 ]
 userVariables []
}
