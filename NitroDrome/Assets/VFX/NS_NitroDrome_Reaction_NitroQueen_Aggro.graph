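# Niagara effect for the NitroDrome Nitro Queen aggro reaction, producing glowing pink-purple flames and embers.
# Three GPU emitters (two mesh flame layers, one sprite ember layer) scale spawn rates by the QueenSpawn NPC parameter, floored by MinSpawnRate.
# Spawn groups share a flattened torus plus six rotated sphere locations whose V position sweeps with system loop age.
# Flames are pulled toward a high attractor, while embers drift on curl noise and flicker.
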NiagaraSystem /NitroDrome/Assets/VFX/NS_NitroDrome_Reaction_NitroQueen_Aggro
dependencies {
 usedDataInterfaces [&NiagaraDataInterfaceCurve, &NiagaraDataInterfaceColorCurve]
 usedDynamicInputs [&MultiplyQuaternion, &MakeQuatFromAxisAngle]
 usedModules [
  &SystemState
  &self:SystemSpawnScript.SystemScriptSource.SystemScriptGraph.NiagaraNodeAssignment_0.SetVariables_#1
 ]
}
emitters {
 FlamesInside {
  simTarget=GPUComputeSim
  emitterUpdateScript [
   &EmitterState()
   &SpawnRate(
    SpawnRate=&Multiply_Float(A=25, B=&ClampFloat(Float=$NPC.NPC_NitroDrome.NitroDrome_QueenSpawn, Min=$System.MinSpawnRate))
    Use Spawn Probability=false
   )
   &SpawnRate(
    SpawnGroup=1
    SpawnRate=&Multiply_Float(A=3, B=&ClampFloat(Float=$NPC.NPC_NitroDrome.NitroDrome_QueenSpawn, Min=$System.MinSpawnRate))
    Use Spawn Probability=false
   ) {module=SpawnRate001}
   &SpawnRate(
    Spawn Probability=2
    SpawnGroup=2
    SpawnRate=&Multiply_Float(A=3, B=&ClampFloat(Float=$NPC.NPC_NitroDrome.NitroDrome_QueenSpawn, Min=$System.MinSpawnRate))
    Use Spawn Probability=false
   ) {module=SpawnRate002}
   &SpawnRate(
    Spawn Probability=2
    SpawnGroup=3
    SpawnRate=&Multiply_Float(A=3, B=&ClampFloat(Float=$NPC.NPC_NitroDrome.NitroDrome_QueenSpawn, Min=$System.MinSpawnRate))
    Use Spawn Probability=false
   ) {module=SpawnRate003}
   &SpawnRate(
    Spawn Probability=2
    SpawnGroup=4
    SpawnRate=&Multiply_Float(A=3, B=&ClampFloat(Float=$NPC.NPC_NitroDrome.NitroDrome_QueenSpawn, Min=$System.MinSpawnRate))
    Use Spawn Probability=false
   ) {module=SpawnRate004}
   &SpawnRate(
    Spawn Probability=2
    SpawnGroup=5
    SpawnRate=&Multiply_Float(A=3, B=&ClampFloat(Float=$NPC.NPC_NitroDrome.NitroDrome_QueenSpawn, Min=$System.MinSpawnRate))
    Use Spawn Probability=false
   ) {module=SpawnRate005}
   &SpawnRate(
    Spawn Probability=2
    SpawnGroup=6
    SpawnRate=&Multiply_Float(A=3, B=&ClampFloat(Float=$NPC.NPC_NitroDrome.NitroDrome_QueenSpawn, Min=$System.MinSpawnRate))
    Use Spawn Probability=false
   ) {module=SpawnRate006}
  ]
  particleSpawnScript [
   &InitializeParticle(
    Color=&ColorFromCurve(
     ColorCurve=DataInterface() {propertyValues {
       AlphaCurve {keys [<at 0 1 RCIM_Cubic>,<at 0.514 1>]}
       BlueCurve {keys [<at 0 0.9874 RCIM_Cubic>,<at 0.4739 1>,<at 1 1 RCIM_Cubic>]}
       GreenCurve {keys [<at 0 0 RCIM_Cubic>,<at 0.4739 0>,<at 1 0.4219 RCIM_Cubic>]}
       RedCurve {keys [<at 0 1 RCIM_Cubic>,<at 0.4739 0.323>,<at 1 0.8098 RCIM_Cubic>]}
      }}
     CurveIndex=&RandomRangeFloat(Minimum=0, Maximum=1)
    )
    Lifetime Max=2.5
    Lifetime Min=2
    Lifetime Mode=::NewEnumerator1
    Mesh Scale Mode=::NewEnumerator1
    Mesh Uniform Scale Max=15
    Mesh Uniform Scale Min=12
    UsePositionOffset=false
    Write Material Random=true
   )
   &ShapeLocation(
    Handle Radius=0
    Large Radius=1000
    MaskSpawnGroup=true
    Non Uniform Scale=<2 1 0>
    Shape Primitive=::NewEnumerator3
    Torus Distribution Mode:ENiagaraTorusDistributionMode=::NewEnumerator1
    UseOwnerScale=true
   ) {module=ShapeLocation001}
   &ShapeLocation(
    MaskSpawnGroup=true
    Non Uniform Scale=<1.3 1 0.5>
    Offset=<0 -500 -2400>
    Offset Mode=::NewEnumerator0
    Rotation Angle=0.325
    Rotation Axis=<0 0 1>
    Rotation Coordinate Space=::Simulation
    Rotation Mode=::NewEnumerator1
    Spawn Group Mask=1
    Sphere Distribution=::NewEnumerator1
    Sphere Radius=5000
    U Position=&RandomRangeFloat(Minimum=0, Maximum=0)
    UseOwnerScale=true
    V Position=&FloatFromCurve(FloatCurve=DataInterface() {propertyValues {Curve {keys [<at 0 0.8>,<at 1 0.99>]}}}, CurveIndex=$System.NormalizedLoopAge)
   ) {module=ShapeLocation002}
   &ShapeLocation(
    MaskSpawnGroup=true
    Non Uniform Scale=<1.6 1 0.5>
    Offset=<0 250 -2400>
    Offset Mode=::NewEnumerator0
    Rotation Angle=0.507
    Rotation Axis=<0 0 1>
    Rotation Coordinate Space=::Simulation
    Rotation Mode=::NewEnumerator1
    Spawn Group Mask=2
    Sphere Distribution=::NewEnumerator1
    Sphere Radius=5000
    U Position=&RandomRangeFloat(Minimum=0, Maximum=0)
    UseOwnerScale=true
    V Position=&FloatFromCurve(FloatCurve=DataInterface() {propertyValues {Curve {keys [<at 0 0.7>,<at 1 0.99>]}}}, CurveIndex=$System.NormalizedLoopAge)
   ) {module=ShapeLocation003}
   &ShapeLocation(
    MaskSpawnGroup=true
    Non Uniform Scale=<1.2 1 0.5>
    Offset=<0 250 -2400>
    Offset Mode=::NewEnumerator0
    Rotation Angle=0.67
    Rotation Axis=<0 0 1>
    Rotation Coordinate Space=::Simulation
    Rotation Mode=::NewEnumerator1
    Spawn Group Mask=3
    Sphere Distribution=::NewEnumerator1
    Sphere Radius=5000
    U Position=&RandomRangeFloat(Minimum=0, Maximum=0)
    UseOwnerScale=true
    V Position=&FloatFromCurve(FloatCurve=DataInterface() {propertyValues {Curve {keys [<at 0 0.7>,<at 1 0.99>]}}}, CurveIndex=$System.NormalizedLoopAge)
   ) {module=ShapeLocation004}
   &ShapeLocation(
    MaskSpawnGroup=true
    Non Uniform Scale=<1.2 1 0.5>
    Offset=<0 250 -2400>
    Offset Mode=::NewEnumerator0
    Rotation Angle=0.82
    Rotation Axis=<0 0 1>
    Rotation Coordinate Space=::Simulation
    Rotation Mode=::NewEnumerator1
    Spawn Group Mask=4
    Sphere Distribution=::NewEnumerator1
    Sphere Radius=5000
    U Position=&RandomRangeFloat(Minimum=0, Maximum=0)
    UseOwnerScale=true
    V Position=&FloatFromCurve(FloatCurve=DataInterface() {propertyValues {Curve {keys [<at 0 0.5>,<at 1 0.99>]}}}, CurveIndex=$System.NormalizedLoopAge)
   ) {module=ShapeLocation005}
   &ShapeLocation(
    MaskSpawnGroup=true
    Non Uniform Scale=<1.6 1 0.5>
    Offset=<0 250 -2500>
    Offset Mode=::NewEnumerator0
    Rotation Angle=0.995
    Rotation Axis=<0 0 1>
    Rotation Coordinate Space=::Simulation
    Rotation Mode=::NewEnumerator1
    Spawn Group Mask=5
    Sphere Distribution=::NewEnumerator1
    Sphere Radius=5000
    U Position=&RandomRangeFloat(Minimum=0, Maximum=0)
    UseOwnerScale=true
    V Position=&FloatFromCurve(FloatCurve=DataInterface() {propertyValues {Curve {keys [<at 0 0.5>,<at 1 0.99>]}}}, CurveIndex=$System.NormalizedLoopAge)
   ) {module=ShapeLocation006}
   &ShapeLocation(
    MaskSpawnGroup=true
    Non Uniform Scale=<1.3 1 0.5>
    Offset=<0 -500 -2500>
    Offset Mode=::NewEnumerator0
    Rotation Angle=0.18
    Rotation Axis=<0 0 1>
    Rotation Coordinate Space=::Simulation
    Rotation Mode=::NewEnumerator1
    Spawn Group Mask=6
    Sphere Distribution=::NewEnumerator1
    Sphere Radius=5000
    U Position=&RandomRangeFloat(Minimum=0, Maximum=0)
    UseOwnerScale=true
    V Position=&FloatFromCurve(FloatCurve=DataInterface() {propertyValues {Curve {keys [<at 0 0.8>,<at 1 0.99>]}}}, CurveIndex=$System.NormalizedLoopAge)
   ) {module=ShapeLocation007}
   &InitialMeshOrientation(Mesh Orientation Mode=::NewEnumerator2, Orientation Vector=<0 0 1>, Use Rotation Vector=false) {enabled=false}
  ]
  particleUpdateScript [
   &ParticleState()
   &ScaleMeshSize(Scale Factor=&Multiply_VectorByFloat(
     Vector=&Multiply_VectorByFloat(
      Vector=<1 1 1>
      Float=&FloatFromCurve(FloatCurve=DataInterface() {propertyValues {Curve {keys [
           {arriveTangent=4.1557, interpMode=RCIM_Cubic, leaveTangent=4.1557, tangentMode=RCTM_User, time=0, value=0}
           {arriveTangent=1.2805, interpMode=RCIM_Cubic, leaveTangent=1.2805, tangentMode=RCTM_User, time=0.3, value=0.75}
           {arriveTangent=-0.3349, interpMode=RCIM_Cubic, leaveTangent=-0.3349, tangentMode=RCTM_User, time=1, value=1.175}
          ]}}})
     )
     Float=1
    ))
   &DynamicMaterialParameters(
    Index 0 Param 1=&FloatFromCurve(
     FloatCurve=DataInterface() {propertyValues {Curve {keys [
         <at 0 0>
         {interpMode=RCIM_Cubic, tangentMode=RCTM_User, time=0.7, value=0}
         {arriveTangent=1.7755, interpMode=RCIM_Cubic, leaveTangent=1.7755, tangentMode=RCTM_User, time=1, value=0.625}
        ]}}}
     Scale Curve=5
    )
    Index 0 Param 2=&FloatFromCurve(
     FloatCurve=DataInterface() {propertyValues {Curve {keys [
         <at 0 0.1>
         {interpMode=RCIM_Cubic, tangentMode=RCTM_User, time=0.6, value=0.65}
         {interpMode=RCIM_Cubic, tangentMode=RCTM_User, time=0.9, value=1.225}
         {interpMode=RCIM_Cubic, tangentMode=RCTM_User, time=1, value=1}
        ]}}}
     Scale Curve=1500
    )
    Index 0 Param 3=1
    Index 0 Param 4=0.35
    Index 1 Param 2=1
    Index 1 Param 4=1
    Index 2 Param 1=1
    Index 2 Param 2=&FloatFromCurve(
     FloatCurve=DataInterface() {propertyValues {Curve {keys [<at 0 0.05>, {interpMode=RCIM_Cubic, tangentMode=RCTM_User, time=0.5, value=0.95}, <at 1 0.7>]}}}
     Scale Curve=2
    )
    Index 2 Param 3=2
    Index 2 Param 4=2
    Index 3 Param 1=&FloatFromCurve(
     FloatCurve=DataInterface() {propertyValues {Curve {keys [
         <at 0 0.25>
         {arriveTangent=1.1168, interpMode=RCIM_Cubic, leaveTangent=1.1168, tangentMode=RCTM_User, time=0.4875, value=0.7}
         {arriveTangent=1.5356, interpMode=RCIM_Cubic, leaveTangent=1.5356, tangentMode=RCTM_User, time=0.7137, value=0.9086}
         <at 1 1>
        ]}}}
     Scale Curve=0.9
    )
    Index 3 Param 2=&RandomRangeFloat(Minimum=0, Maximum=1)
    Module.Param0WriteEnabled=true
    Module.Param0WriteEnabled1=true
    Module.Param0WriteEnabled2=true
    Module.Param0WriteEnabled3=true
    Module.Param1WriteEnabled=true
    Module.Param1WriteEnabled1=true
    Module.Param1WriteEnabled2=true
    Module.Param1WriteEnabled3=true
    Module.Param2WriteEnabled=true
    Module.Param2WriteEnabled1=true
    Module.Param2WriteEnabled2=true
    Module.Param2WriteEnabled3=true
    Module.Param3WriteEnabled=true
    Module.Param3WriteEnabled1=true
    Module.Param3WriteEnabled2=true
    Module.Param3WriteEnabled3=true
    Write Parameter Index 1=true
    Write Parameter Index 2=true
    Write Parameter Index 3=true
   ) {module=DynamicMaterialParameters001}
   &ScaleColor(
    Scale Alpha=&FloatFromCurve(
     FloatCurve=DataInterface() {propertyValues {Curve {keys [{interpMode=RCIM_Cubic, tangentMode=RCTM_User, time=0, value=0}, <at 0.2 1 RCIM_Cubic>, {tangentMode=RCTM_User, time=1, value=0}]}}}
     Scale Curve=2
    )
    Scale RGB=&Multiply_VectorByFloat(
     Vector=<500 500 500>
     Float=&FloatFromCurve(
      FloatCurve=DataInterface() {propertyValues {Curve {keys [
          {interpMode=RCIM_Cubic, tangentMode=RCTM_User, time=0, value=0.2}
          {interpMode=RCIM_Cubic, tangentMode=RCTM_User, time=0.2, value=1.975}
          <at 1 0.4>
         ]}}}
      Scale Curve=2
     )
    )
    ScaleA=true
    ScaleRGB=true
   )
   /SetVariables(
    Particles.CamOff2=5
    Particles.Linear Color=&MultiplyLinearColor(Linear Color=$Particles.Color, Scale Factor=<1 1 1 25>)
    Particles.MaskedScale:Vector3f=&Multiply_Vector(A=$Particles.Scale, B=<0.85 1 1>)
   ) {isSetParameters=true, module=SetVariables_#2}
   &PointAttractionForce(
    Attraction Radius=150000
    Attractor Position Offset:Vector3f=<0 0 10000>
    Kill Within Radius=false
    Module.Use Falloff=true
    UseAttractorOffset=true
   )
   &SolveForcesAndVelocity(Clamp Velocity=false, Limit Acceleration=false)
  ]
  rendererClasses [&NiagaraMeshRendererProperties]
  renderers [
   {rendererClass=&NiagaraMeshRendererProperties, rendererIndex=0}
   {rendererClass=&NiagaraMeshRendererProperties, rendererIndex=1}
  ]
 }
 FlamesInside001 {
  simTarget=GPUComputeSim
  emitterUpdateScript [
   &EmitterState()
   &SpawnRate(
    SpawnRate=&Multiply_Float(A=25, B=&ClampFloat(Float=$NPC.NPC_NitroDrome.NitroDrome_QueenSpawn, Min=$System.MinSpawnRate))
    Use Spawn Probability=false
   )
   &SpawnRate(
    SpawnGroup=1
    SpawnRate=&Multiply_Float(A=3, B=&ClampFloat(Float=$NPC.NPC_NitroDrome.NitroDrome_QueenSpawn, Min=$System.MinSpawnRate))
    Use Spawn Probability=false
   ) {module=SpawnRate001}
   &SpawnRate(
    Spawn Probability=2
    SpawnGroup=2
    SpawnRate=&Multiply_Float(A=3, B=&ClampFloat(Float=$NPC.NPC_NitroDrome.NitroDrome_QueenSpawn, Min=$System.MinSpawnRate))
    Use Spawn Probability=false
   ) {module=SpawnRate002}
   &SpawnRate(
    Spawn Probability=2
    SpawnGroup=3
    SpawnRate=&Multiply_Float(A=3, B=&ClampFloat(Float=$NPC.NPC_NitroDrome.NitroDrome_QueenSpawn, Min=$System.MinSpawnRate))
    Use Spawn Probability=false
   ) {module=SpawnRate003}
   &SpawnRate(
    Spawn Probability=2
    SpawnGroup=4
    SpawnRate=&Multiply_Float(A=3, B=&ClampFloat(Float=$NPC.NPC_NitroDrome.NitroDrome_QueenSpawn, Min=$System.MinSpawnRate))
    Use Spawn Probability=false
   ) {module=SpawnRate004}
   &SpawnRate(
    Spawn Probability=2
    SpawnGroup=5
    SpawnRate=&Multiply_Float(A=3, B=&ClampFloat(Float=$NPC.NPC_NitroDrome.NitroDrome_QueenSpawn, Min=$System.MinSpawnRate))
    Use Spawn Probability=false
   ) {module=SpawnRate005}
   &SpawnRate(
    Spawn Probability=2
    SpawnGroup=6
    SpawnRate=&Multiply_Float(A=3, B=&ClampFloat(Float=$NPC.NPC_NitroDrome.NitroDrome_QueenSpawn, Min=$System.MinSpawnRate))
    Use Spawn Probability=false
   ) {module=SpawnRate006}
  ]
  particleSpawnScript [
   &InitializeParticle(
    Color=<rgba 18.1368 1.2449 25 1>
    Lifetime Max=2.5
    Lifetime Min=2
    Lifetime Mode=::NewEnumerator1
    Mesh Scale=<2 1 1>
    Mesh Scale Mode=::NewEnumerator4
    UsePositionOffset=false
    Write Material Random=true
   )
   &SystemLocation(MaskSpawnGroup=false, Offset=<0 0 200>)
   &ShapeLocation(
    Handle Radius=0
    Large Radius=1000
    MaskSpawnGroup=true
    Non Uniform Scale=<2 1 0>
    Shape Primitive=::NewEnumerator3
    Torus Distribution Mode:ENiagaraTorusDistributionMode=::NewEnumerator1
    UseOwnerScale=true
   ) {module=ShapeLocation001}
   &ShapeLocation(
    MaskSpawnGroup=true
    Non Uniform Scale=<1.3 1 0.5>
    Offset=<0 -500 -2400>
    Offset Mode=::NewEnumerator0
    Rotation Angle=0.325
    Rotation Axis=<0 0 1>
    Rotation Coordinate Space=::Simulation
    Rotation Mode=::NewEnumerator1
    Spawn Group Mask=1
    Sphere Distribution=::NewEnumerator1
    Sphere Radius=5000
    U Position=&RandomRangeFloat(Minimum=0, Maximum=0)
    UseOwnerScale=true
    V Position=&FloatFromCurve(FloatCurve=DataInterface() {propertyValues {Curve {keys [<at 0 0.8>,<at 1 0.99>]}}}, CurveIndex=$System.NormalizedLoopAge)
   ) {module=ShapeLocation002}
   &ShapeLocation(
    MaskSpawnGroup=true
    Non Uniform Scale=<1.6 1 0.5>
    Offset=<0 250 -2400>
    Offset Mode=::NewEnumerator0
    Rotation Angle=0.507
    Rotation Axis=<0 0 1>
    Rotation Coordinate Space=::Simulation
    Rotation Mode=::NewEnumerator1
    Spawn Group Mask=2
    Sphere Distribution=::NewEnumerator1
    Sphere Radius=5000
    U Position=&RandomRangeFloat(Minimum=0, Maximum=0)
    UseOwnerScale=true
    V Position=&FloatFromCurve(FloatCurve=DataInterface() {propertyValues {Curve {keys [<at 0 0.7>,<at 1 0.99>]}}}, CurveIndex=$System.NormalizedLoopAge)
   ) {module=ShapeLocation003}
   &ShapeLocation(
    MaskSpawnGroup=true
    Non Uniform Scale=<1.2 1 0.5>
    Offset=<0 250 -2400>
    Offset Mode=::NewEnumerator0
    Rotation Angle=0.67
    Rotation Axis=<0 0 1>
    Rotation Coordinate Space=::Simulation
    Rotation Mode=::NewEnumerator1
    Spawn Group Mask=3
    Sphere Distribution=::NewEnumerator1
    Sphere Radius=5000
    U Position=&RandomRangeFloat(Minimum=0, Maximum=0)
    UseOwnerScale=true
    V Position=&FloatFromCurve(FloatCurve=DataInterface() {propertyValues {Curve {keys [<at 0 0.7>,<at 1 0.99>]}}}, CurveIndex=$System.NormalizedLoopAge)
   ) {module=ShapeLocation004}
   &ShapeLocation(
    MaskSpawnGroup=true
    Non Uniform Scale=<1.2 1 0.5>
    Offset=<0 250 -2400>
    Offset Mode=::NewEnumerator0
    Rotation Angle=0.82
    Rotation Axis=<0 0 1>
    Rotation Coordinate Space=::Simulation
    Rotation Mode=::NewEnumerator1
    Spawn Group Mask=4
    Sphere Distribution=::NewEnumerator1
    Sphere Radius=5000
    U Position=&RandomRangeFloat(Minimum=0, Maximum=0)
    UseOwnerScale=true
    V Position=&FloatFromCurve(FloatCurve=DataInterface() {propertyValues {Curve {keys [<at 0 0.5>,<at 1 0.99>]}}}, CurveIndex=$System.NormalizedLoopAge)
   ) {module=ShapeLocation005}
   &ShapeLocation(
    MaskSpawnGroup=true
    Non Uniform Scale=<1.6 1 0.5>
    Offset=<0 250 -2500>
    Offset Mode=::NewEnumerator0
    Rotation Angle=0.995
    Rotation Axis=<0 0 1>
    Rotation Coordinate Space=::Simulation
    Rotation Mode=::NewEnumerator1
    Spawn Group Mask=5
    Sphere Distribution=::NewEnumerator1
    Sphere Radius=5000
    U Position=&RandomRangeFloat(Minimum=0, Maximum=0)
    UseOwnerScale=true
    V Position=&FloatFromCurve(FloatCurve=DataInterface() {propertyValues {Curve {keys [<at 0 0.5>,<at 1 0.99>]}}}, CurveIndex=$System.NormalizedLoopAge)
   ) {module=ShapeLocation006}
   &ShapeLocation(
    MaskSpawnGroup=true
    Non Uniform Scale=<1.3 1 0.5>
    Offset=<0 -500 -2500>
    Offset Mode=::NewEnumerator0
    Rotation Angle=0.18
    Rotation Axis=<0 0 1>
    Rotation Coordinate Space=::Simulation
    Rotation Mode=::NewEnumerator1
    Spawn Group Mask=6
    Sphere Distribution=::NewEnumerator1
    Sphere Radius=5000
    U Position=&RandomRangeFloat(Minimum=0, Maximum=0)
    UseOwnerScale=true
    V Position=&FloatFromCurve(FloatCurve=DataInterface() {propertyValues {Curve {keys [<at 0 0.8>,<at 1 0.99>]}}}, CurveIndex=$System.NormalizedLoopAge)
   ) {module=ShapeLocation007}
   &InitialMeshOrientation(Mesh Orientation Mode=::NewEnumerator2, Orientation Vector=<0 0 1>, Use Rotation Vector=false) {enabled=false}
  ]
  particleUpdateScript [
   &ParticleState()
   &ScaleMeshSize(Scale Factor=&Multiply_VectorByFloat(
     Vector=&Multiply_VectorByFloat(
      Vector=<2 0.5 0.5>
      Float=&FloatFromCurve(FloatCurve=DataInterface() {propertyValues {Curve {keys [
           {arriveTangent=4.1557, interpMode=RCIM_Cubic, leaveTangent=4.1557, tangentMode=RCTM_User, time=0, value=0}
           {arriveTangent=1.2805, interpMode=RCIM_Cubic, leaveTangent=1.2805, tangentMode=RCTM_User, time=0.3, value=0.75}
           {arriveTangent=-0.3349, interpMode=RCIM_Cubic, leaveTangent=-0.3349, tangentMode=RCTM_User, time=1, value=1.175}
          ]}}})
     )
     Float=1
    ))
   &DynamicMaterialParameters(
    Index 0 Param 1=&FloatFromCurve(
     FloatCurve=DataInterface() {propertyValues {Curve {keys [
         <at 0 0>
         {interpMode=RCIM_Cubic, tangentMode=RCTM_User, time=0.7, value=0}
         {arriveTangent=1.7755, interpMode=RCIM_Cubic, leaveTangent=1.7755, tangentMode=RCTM_User, time=1, value=0.625}
        ]}}}
     Scale Curve=25
    )
    Index 0 Param 2=&FloatFromCurve(
     FloatCurve=DataInterface() {propertyValues {Curve {keys [
         <at 0 0.1>
         {interpMode=RCIM_Cubic, tangentMode=RCTM_User, time=0.6, value=0.65}
         {interpMode=RCIM_Cubic, tangentMode=RCTM_User, time=0.9, value=1.225}
         {interpMode=RCIM_Cubic, tangentMode=RCTM_User, time=1, value=1}
        ]}}}
     Scale Curve=500
    )
    Index 0 Param 3=1
    Index 0 Param 4=0.35
    Index 1 Param 2=1
    Index 1 Param 4=1
    Index 2 Param 1=1
    Index 2 Param 2=&FloatFromCurve(
     FloatCurve=DataInterface() {propertyValues {Curve {keys [<at 0 0.05>, {interpMode=RCIM_Cubic, tangentMode=RCTM_User, time=0.5, value=0.95}, <at 1 0.7>]}}}
     Scale Curve=2
    )
    Index 2 Param 3=1
    Index 2 Param 4=1
    Index 3 Param 1=&FloatFromCurve(
     FloatCurve=DataInterface() {propertyValues {Curve {keys [
         <at 0 0.25>
         {arriveTangent=1.1168, interpMode=RCIM_Cubic, leaveTangent=1.1168, tangentMode=RCTM_User, time=0.4875, value=0.7}
         {arriveTangent=1.5356, interpMode=RCIM_Cubic, leaveTangent=1.5356, tangentMode=RCTM_User, time=0.7137, value=0.9086}
         <at 1 1>
        ]}}}
     Scale Curve=0.9
    )
    Index 3 Param 2=&RandomRangeFloat(Minimum=0, Maximum=1)
    Module.Param0WriteEnabled=true
    Module.Param0WriteEnabled1=true
    Module.Param0WriteEnabled2=true
    Module.Param0WriteEnabled3=true
    Module.Param1WriteEnabled=true
    Module.Param1WriteEnabled1=true
    Module.Param1WriteEnabled2=true
    Module.Param1WriteEnabled3=true
    Module.Param2WriteEnabled=true
    Module.Param2WriteEnabled1=true
    Module.Param2WriteEnabled2=true
    Module.Param2WriteEnabled3=true
    Module.Param3WriteEnabled=true
    Module.Param3WriteEnabled1=true
    Module.Param3WriteEnabled2=true
    Module.Param3WriteEnabled3=true
    Write Parameter Index 1=true
    Write Parameter Index 2=true
    Write Parameter Index 3=true
   ) {module=DynamicMaterialParameters001}
   &ScaleColor(
    Scale Alpha=&FloatFromCurve(
     FloatCurve=DataInterface() {propertyValues {Curve {keys [
         {interpMode=RCIM_Cubic, tangentMode=RCTM_User, time=0, value=0}
         {interpMode=RCIM_Cubic, tangentMode=RCTM_User, time=0.5, value=1}
         {tangentMode=RCTM_User, time=1, value=0}
        ]}}}
     Scale Curve=2
    )
    Scale RGB=&Multiply_VectorByFloat(
     Vector=<15 15 15>
     Float=&FloatFromCurve(FloatCurve=DataInterface() {propertyValues {Curve {keys [
          {interpMode=RCIM_Cubic, tangentMode=RCTM_User, time=0, value=0.2}
          {interpMode=RCIM_Cubic, tangentMode=RCTM_User, time=0.5, value=2}
          <at 1 0.4>
         ]}}})
    )
    ScaleA=true
    ScaleRGB=true
   )
   /SetVariables(
    Particles.CamOff2=250
    Particles.Linear Color=&MultiplyLinearColor(Linear Color=$Particles.Color, Scale Factor=<1 1 1 50>)
    Particles.MaskedScale:Vector3f=&Multiply_Vector(A=$Particles.Scale, B=<1 1 1>)
   ) {isSetParameters=true, module=SetVariables_#2}
   &PointAttractionForce(
    Attraction Radius=150000
    Attraction Strength=0.1
    Attractor Position Offset:Vector3f=<0 0 10000>
    Kill Within Radius=false
    Module.Use Falloff=true
    UseAttractorOffset=true
   )
   &Drag(Drag=0, Use Linear Drag=true, Use Rotational Drag=true)
   &SolveForcesAndVelocity(Clamp Velocity=false, Limit Acceleration=false)
  ]
  rendererClasses [&NiagaraMeshRendererProperties]
  renderers [{rendererClass=&NiagaraMeshRendererProperties, rendererIndex=0}]
 }
 TrailEmbers_EPIC {
  simTarget=GPUComputeSim
  emitterUpdateScript [
   &EmitterState()
   &SpawnRate(
    SpawnRate=&Multiply_Float(A=2500, B=&ClampFloat(Float=$NPC.NPC_NitroDrome.NitroDrome_QueenSpawn, Min=$System.MinSpawnRate))
    Use Spawn Probability=false
   )
   &SpawnRate(
    SpawnGroup=1
    SpawnRate=&Multiply_Float(A=250, B=&ClampFloat(Float=$NPC.NPC_NitroDrome.NitroDrome_QueenSpawn, Min=$System.MinSpawnRate))
    Use Spawn Probability=false
   ) {module=SpawnRate001}
   &SpawnRate(
    Spawn Probability=2
    SpawnGroup=2
    SpawnRate=&Multiply_Float(A=250, B=&ClampFloat(Float=$NPC.NPC_NitroDrome.NitroDrome_QueenSpawn, Min=$System.MinSpawnRate))
    Use Spawn Probability=false
   ) {module=SpawnRate002}
   &SpawnRate(
    Spawn Probability=2
    SpawnGroup=3
    SpawnRate=&Multiply_Float(A=250, B=&ClampFloat(Float=$NPC.NPC_NitroDrome.NitroDrome_QueenSpawn, Min=$System.MinSpawnRate))
    Use Spawn Probability=false
   ) {module=SpawnRate003}
   &SpawnRate(
    Spawn Probability=2
    SpawnGroup=4
    SpawnRate=&Multiply_Float(A=250, B=&ClampFloat(Float=$NPC.NPC_NitroDrome.NitroDrome_QueenSpawn, Min=$System.MinSpawnRate))
    Use Spawn Probability=false
   ) {module=SpawnRate004}
   &SpawnRate(
    Spawn Probability=2
    SpawnGroup=5
    SpawnRate=&Multiply_Float(A=250, B=&ClampFloat(Float=$NPC.NPC_NitroDrome.NitroDrome_QueenSpawn, Min=$System.MinSpawnRate))
    Use Spawn Probability=false
   ) {module=SpawnRate005}
   &SpawnRate(
    Spawn Probability=2
    SpawnGroup=6
    SpawnRate=&Multiply_Float(A=250, B=&ClampFloat(Float=$NPC.NPC_NitroDrome.NitroDrome_QueenSpawn, Min=$System.MinSpawnRate))
    Use Spawn Probability=false
   ) {module=SpawnRate006}
  ]
  particleSpawnScript [
   &InitializeParticle(
    Color=&RandomRangeLinearColor(Minimum=<rgba 25 0 20.9407 1>, Maximum=<rgba 68.4564 0 250 1>)
    Lifetime Max=4
    Lifetime Min=0.25
    Lifetime Mode=::NewEnumerator1
    Sprite Rotation Mode=::NewEnumerator2
    Sprite Size Mode=::NewEnumerator1
    Uniform Sprite Size Max=100
    Uniform Sprite Size Min=25
    UsePositionOffset=false
    Write Material Random=true
   )
   &AddVelocity(
    Cone Angle=1
    Cone Axis=&MakeVector(X=&Sine(Normalized Angle=$Emitter.Age, Period=5, Scale=1, Bias=0), Y=0, Z=1)
    Module.UseConeFalloff=false
    Module.UseVelDistribution=false
    Override Seed=false
    Velocity Mode=::NewEnumerator2
    Velocity Speed=500
   )
   &SystemLocation(MaskSpawnGroup=false, Offset=<0 0 200>)
   &ShapeLocation(
    Handle Radius=0
    Large Radius=1000
    MaskSpawnGroup=true
    Non Uniform Scale=<2 1 0>
    Shape Primitive=::NewEnumerator3
    Torus Distribution Mode:ENiagaraTorusDistributionMode=::NewEnumerator1
    UseOwnerScale=true
   ) {module=ShapeLocation001}
   &ShapeLocation(
    MaskSpawnGroup=true
    Non Uniform Scale=<1.3 1 0.5>
    Offset=<0 -500 -2400>
    Offset Mode=::NewEnumerator0
    Rotation Angle=0.325
    Rotation Axis=<0 0 1>
    Rotation Coordinate Space=::Simulation
    Rotation Mode=::NewEnumerator1
    Spawn Group Mask=1
    Sphere Distribution=::NewEnumerator1
    Sphere Radius=5000
    U Position=&RandomRangeFloat(Minimum=0, Maximum=0)
    UseOwnerScale=true
    V Position=&FloatFromCurve(FloatCurve=DataInterface() {propertyValues {Curve {keys [<at 0 0.8>,<at 1 0.99>]}}}, CurveIndex=$System.NormalizedLoopAge)
   ) {module=ShapeLocation002}
   &ShapeLocation(
    MaskSpawnGroup=true
    Non Uniform Scale=<1.6 1 0.5>
    Offset=<0 250 -2400>
    Offset Mode=::NewEnumerator0
    Rotation Angle=0.507
    Rotation Axis=<0 0 1>
    Rotation Coordinate Space=::Simulation
    Rotation Mode=::NewEnumerator1
    Spawn Group Mask=2
    Sphere Distribution=::NewEnumerator1
    Sphere Radius=5000
    U Position=&RandomRangeFloat(Minimum=0, Maximum=0)
    UseOwnerScale=true
    V Position=&FloatFromCurve(FloatCurve=DataInterface() {propertyValues {Curve {keys [<at 0 0.7>,<at 1 0.99>]}}}, CurveIndex=$System.NormalizedLoopAge)
   ) {module=ShapeLocation003}
   &ShapeLocation(
    MaskSpawnGroup=true
    Non Uniform Scale=<1.2 1 0.5>
    Offset=<0 250 -2400>
    Offset Mode=::NewEnumerator0
    Rotation Angle=0.67
    Rotation Axis=<0 0 1>
    Rotation Coordinate Space=::Simulation
    Rotation Mode=::NewEnumerator1
    Spawn Group Mask=3
    Sphere Distribution=::NewEnumerator1
    Sphere Radius=5000
    U Position=&RandomRangeFloat(Minimum=0, Maximum=0)
    UseOwnerScale=true
    V Position=&FloatFromCurve(FloatCurve=DataInterface() {propertyValues {Curve {keys [<at 0 0.7>,<at 1 0.99>]}}}, CurveIndex=$System.NormalizedLoopAge)
   ) {module=ShapeLocation004}
   &ShapeLocation(
    MaskSpawnGroup=true
    Non Uniform Scale=<1.2 1 0.5>
    Offset=<0 250 -2400>
    Offset Mode=::NewEnumerator0
    Rotation Angle=0.82
    Rotation Axis=<0 0 1>
    Rotation Coordinate Space=::Simulation
    Rotation Mode=::NewEnumerator1
    Spawn Group Mask=4
    Sphere Distribution=::NewEnumerator1
    Sphere Radius=5000
    U Position=&RandomRangeFloat(Minimum=0, Maximum=0)
    UseOwnerScale=true
    V Position=&FloatFromCurve(FloatCurve=DataInterface() {propertyValues {Curve {keys [<at 0 0.5>,<at 1 0.99>]}}}, CurveIndex=$System.NormalizedLoopAge)
   ) {module=ShapeLocation005}
   &ShapeLocation(
    MaskSpawnGroup=true
    Non Uniform Scale=<1.6 1 0.5>
    Offset=<0 250 -2500>
    Offset Mode=::NewEnumerator0
    Rotation Angle=0.995
    Rotation Axis=<0 0 1>
    Rotation Coordinate Space=::Simulation
    Rotation Mode=::NewEnumerator1
    Spawn Group Mask=5
    Sphere Distribution=::NewEnumerator1
    Sphere Radius=5000
    U Position=&RandomRangeFloat(Minimum=0, Maximum=0)
    UseOwnerScale=true
    V Position=&FloatFromCurve(FloatCurve=DataInterface() {propertyValues {Curve {keys [<at 0 0.5>,<at 1 0.99>]}}}, CurveIndex=$System.NormalizedLoopAge)
   ) {module=ShapeLocation006}
   &ShapeLocation(
    MaskSpawnGroup=true
    Non Uniform Scale=<1.3 1 0.5>
    Offset=<0 -500 -2500>
    Offset Mode=::NewEnumerator0
    Rotation Angle=0.18
    Rotation Axis=<0 0 1>
    Rotation Coordinate Space=::Simulation
    Rotation Mode=::NewEnumerator1
    Spawn Group Mask=6
    Sphere Distribution=::NewEnumerator1
    Sphere Radius=5000
    U Position=&RandomRangeFloat(Minimum=0, Maximum=0)
    UseOwnerScale=true
    V Position=&FloatFromCurve(FloatCurve=DataInterface() {propertyValues {Curve {keys [<at 0 0.8>,<at 1 0.99>]}}}, CurveIndex=$System.NormalizedLoopAge)
   ) {module=ShapeLocation007}
   &ShapeLocation(MaskSpawnGroup=false, Override Seed=false, Sphere Radius=500, UseOwnerScale=true)
  ]
  particleUpdateScript [
   /SetVariables(Particles.VisibilityTag=&SpawnGroup()) {enabled=false, isSetParameters=true}
   &Drag(
    Drag=&FloatFromCurve(FloatCurve=DataInterface() {propertyValues {Curve {keys [<at 0 0>,<at 0.2 0.95>,<at 1 0>]}}}, Scale Curve=0)
    Use Linear Drag=true
    Use Rotational Drag=false
   )
   &GravityForce()
   &CurlNoiseForce(
    Noise Frequency=&RandomRangeFloat(Minimum=0, Maximum=2)
    Noise Strength=&FloatFromCurve(
     FloatCurve=DataInterface() {propertyValues {Curve {keys [<at 0 1>,<at 1 0.5>]}}}
     Scale Curve=&RandomRangeFloat(Minimum=250, Maximum=500)
    )
    Pan Noise=false
   )
   &SolveForcesAndVelocity(Acceleration Limit=15, Clamp Velocity=false, Limit Acceleration=false, Speed Limit=1500)
   &ParticleState()
   &DynamicMaterialParameters(
    Index 0 Param 1=0.8
    Index 0 Param 2=1
    Index 0 Param 3=1
    Index 1 Param 1=150
    Index 1 Param 4=1
    Module.Param0WriteEnabled=true
    Module.Param0WriteEnabled1=true
    Module.Param1WriteEnabled=true
    Module.Param1WriteEnabled1=true
    Module.Param2WriteEnabled=true
    Module.Param2WriteEnabled1=true
    Module.Param3WriteEnabled=false
    Module.Param3WriteEnabled1=true
    Write Parameter Index 1=true
   )
   &ScaleColor(
    Scale RGB=&Multiply_VectorByFloat(
     Vector=<250 250 250>
     Float=&FloatFromCurve(
      FloatCurve=DataInterface() {propertyValues {Curve {keys [{interpMode=RCIM_Cubic, tangentMode=RCTM_User, time=0, value=1}, {interpMode=RCIM_Cubic, tangentMode=RCTM_User, time=1, value=0}]}}}
      Scale Curve=&Waveform(
       "Waveform [1]"=::NewEnumerator0
       "[1] Phase Scale"=&RandomRangeFloat(Minimum=1, Maximum=2)
       "[1] Frequency"=&Lerp_Float(
        A=&RandomRangeFloat(Minimum=1, Maximum=2)
        B=1
        Alpha=&MaskFloatBySpawnGroup(Masked Float=1, Passthrough Float=1, Spawn Group Mask=1)
       )
       "[1] Amplitude Min/Max"=<0.5 1>
      )
     )
    )
    ScaleA=true
    ScaleRGB=true
   ) {module=ScaleColor001}
   &ScaleColor(Scale Alpha=0.02, Scale RGB=<0.02 0.01 0.01>, ScaleA=true, ScaleRGB=true) {enabled=false}
   /Niagara/Modules/Update/Camera/MaintainInCameraParticleScale(Pixel Depth=0.1, Render Target Width=0.1)
   &ScaleSpriteSizeBySpeed(
    Max Scale Factor=<0.2 0.5>
    Min Scale Factor=<0.15 0.15>
    Sample Scale Factor By Curve=true
    Scale Factor Curve=DataInterface() {propertyValues {Curve {keys [
        {interpMode=RCIM_Cubic, tangentMode=RCTM_User, time=0, value=0}
        {arriveTangent=2.9664, interpMode=RCIM_Cubic, leaveTangent=2.9664, tangentMode=RCTM_User, time=1, value=1}
       ]}}}
    Velocity Threshold=100
   )
   &ScaleSpriteSize(
    Uniform Curve Scale=&Multiply_Float(
     A=&Multiply_Float(
      A=0.2
      B=&Waveform(
       "Waveform [1]"=::NewEnumerator0
       "[1] Phase Scale"=&RandomRangeFloat(Minimum=3, Maximum=10)
       "[1] Frequency"=&RandomRangeFloat(Minimum=0.2, Maximum=0.5)
       "[1] Amplitude Min/Max"=<0 1>
      )
     )
     B=&FN_LOD_Distance_Remap(Start Distance=500, End Distance=5000, Minimum Output=0.5, Maximum Output=1, Clamp Output=true)
    )
    Uniform Curve Sprite Scale=DataInterface() {propertyValues {Curve {keys [
        {arriveTangent=2.9868, interpMode=RCIM_Cubic, leaveTangent=2.9868, tangentMode=RCTM_User, time=0, value=1.65}
        <at 0.4 2.3>
        {arriveTangent=-0.0405, interpMode=RCIM_Cubic, leaveTangent=-0.0405, tangentMode=RCTM_User, time=1, value=1.4}
       ]}}}
   )
   /SetVariables(
    Particles.Linear Color=&MultiplyLinearColor(Linear Color=$Particles.Color, Scale Factor=<0.02 0.02 0.02 0.0001>)
    Particles.Vector:Vector3f=&Multiply_Vector(A=$Particles.Velocity, B=<1 1 1>)
    Particles.Vector 2D:Vector2f=&Multiply_Vector2D(A=<1 4>, B=$Particles.SpriteSize)
   ) {isSetParameters=true}
  ]
  rendererClasses [&NiagaraSpriteRendererProperties]
  renderers [
   {rendererClass=&NiagaraSpriteRendererProperties}
   {rendererClass=&NiagaraSpriteRendererProperties, rendererIndex=1}
  ]
 }
}
summary {
 systemName=NS_NitroDrome_Reaction_NitroQueen_Aggro
 emitters [
  {emitterName=TrailEmbers_EPIC, simTarget=GPUComputeSim, rendererClasses [&NiagaraSpriteRendererProperties]}
  {emitterName=FlamesInside, simTarget=GPUComputeSim, rendererClasses [&NiagaraMeshRendererProperties]}
  {emitterName=FlamesInside001, simTarget=GPUComputeSim, rendererClasses [&NiagaraMeshRendererProperties]}
 ]
 userVariables []
}
